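# Physics asset for the front face-accessory part of the LivelySugar medium female player cosmetic, simulating left and right dynamic hair chains.
# Capsule bodies cover the spine, clavicles, upper arms, head and hair bones, with torso, arm and head joints fully locked.
# Two seven-bone front chains hang from the head on narrow swing and twist limits.
# Short back-hair chains on each side are locked rigidly to the head.
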PhysicsAsset /BRCosmetics/Characters/Player/Female/Medium/Bodies/F_MED_LivelySugar/Meshes/Parts/F_MED_FaceAcc_LivelySugar_front_Physics
bodies {
 clavicle_l [{
   shapeType=Capsule
   length=7.6779
   radius=4.1717
   shapeName=clavicle_l_capsule
   center=<7.2606 -0.7215 0.483>
   rotation=<rot 83.0357 -179.9977 -179.998>
  }]
 clavicle_r [{
   shapeType=Capsule
   length=4.1779
   radius=4.1717
   shapeName=clavicle_l_capsule
   center=<-8.3401 0.6246 -1.081>
   rotation=<rot 83.0357 -179.9971 0.0026>
  }]
 dyn_hair_bk_l_01 [{
   shapeType=Capsule
   length=2.4654
   radius=3
   shapeName=dyn_hair_bk_l_01_capsule
   center=<-1.5423 -0.0087 0.4224>
   rotation=<rot 1.013 91.6142 95.7553>
  }]
 dyn_hair_bk_l_02 [{
   shapeType=Capsule
   length=3.5
   radius=3.5
   shapeName=dyn_hair_bk_l_02_capsule
   center=<2.1673 0.7041 -2.7707>
   rotation=<rot 0 -79.1315 109.3189>
  }]
 dyn_hair_bk_r_01 [{
   shapeType=Capsule
   length=2.4654
   radius=3
   shapeName=dyn_hair_bk_r_01_capsule
   center=<1.5423 0.0087 -0.4224>
   rotation=<rot 1.013 91.6142 -84.2447>
  }]
 dyn_hair_bk_r_02 [{
   shapeType=Capsule
   length=3.5
   radius=3.5
   shapeName=dyn_hair_bk_r_02_capsule
   center=<3.0423 -0.0704 -3.1454>
   rotation=<rot 3.2933 -88.5789 -70.9532>
  }]
 dyn_hair_l_01 [{shapeType=Capsule, length=0.505, radius=0.505, shapeName=dyn_hair_l_01_capsule, rotation=<rot 0 -90 0>}]
 dyn_hair_l_02 [{
   shapeType=Capsule
   length=4.7053
   radius=1.8039
   shapeName=dyn_hair_l_02_capsule
   center=<4.3135 0 0>
   rotation=<rot 9.6238 -88.2819 95.7175>
  }]
 dyn_hair_l_03 [{
   shapeType=Capsule
   length=1
   radius=2.4594
   shapeName=dyn_hair_l_03_capsule
   center=<2.7514 0.6272 -0.666>
   rotation=<rot -22.4139 -90.0004 89.6993>
  }]
 dyn_hair_l_04 [{
   shapeType=Capsule
   length=4.4891
   radius=1.5
   shapeName=dyn_hair_l_04_capsule
   center=<2.0823 0.2065 -0.1756>
   rotation=<rot -14.485 -90.0003 89.5318>
  }]
 dyn_hair_l_05 [{
   shapeType=Capsule
   length=2.4565
   radius=1.7
   shapeName=dyn_hair_l_05_capsule
   center=<2.3417 -0.3724 0.0699>
   rotation=<rot -9.5674 90 -83.0668>
  }]
 dyn_hair_l_06 [{
   shapeType=Capsule
   length=4.8022
   radius=1.7217
   shapeName=dyn_hair_l_06_capsule
   center=<2.8732 -0.5 0.4>
   rotation=<rot -1.1865 119.722 -87.9442>
  }]
 dyn_hair_l_07 [{shapeType=Capsule, length=3.951, radius=2, shapeName=dyn_hair_l_07_capsule, center=<1.8942 -2.5 2.5>, rotation=<rot 0 83.2812 -80>}]
 dyn_hair_r_01 [{
   shapeType=Capsule
   length=2.465
   radius=2
   shapeName=dyn_hair_r_01_capsule
   center=<4.1027 -1.3455 0.3174>
   rotation=<rot 1.5824 94.9283 -95.6467>
  }]
 dyn_hair_r_02 [{
   shapeType=Capsule
   length=3.1155
   radius=1.6738
   shapeName=dyn_hair_r_02_capsule
   center=<3.6718 -0.6639 -0.9959>
   rotation=<rot -5.8083 91.3066 -94.1698>
  }]
 dyn_hair_r_03 [{
   shapeType=Capsule
   length=2.7592
   radius=1.3932
   shapeName=dyn_hair_r_03_capsule
   center=<1.37 0.0356 -0.4471>
   rotation=<rot 26.0752 109.2074 -79.9999>
  }]
 dyn_hair_r_04 [{
   shapeType=Capsule
   length=3.714
   radius=2.111
   shapeName=dyn_hair_r_04_capsule
   center=<1.4191 -0.6696 -0.6725>
   rotation=<rot 22.5546 -84.6146 -96.8399>
  }]
 dyn_hair_r_05 [{shapeType=Capsule, length=2, radius=1.9, shapeName=dyn_hair_r_05_capsule, center=<1.7567 -1.5 0.28>, rotation=<rot 54.6485 88.2121 -93.352>}]
 dyn_hair_r_06 [{
   shapeType=Capsule
   length=3.5577
   radius=2.1
   shapeName=dyn_hair_r_06_capsule
   center=<1.9723 -0.7 0.2>
   rotation=<rot 21.9413 -54.0003 -90.2441>
  }]
 dyn_hair_r_07 [{
   shapeType=Capsule
   length=3.5578
   radius=1.9
   shapeName=dyn_hair_r_07_capsule
   center=<3.0523 -0.4033 -0.8976>
   rotation=<rot 31.0787 -59.0172 -63.9954>
  }]
 head [{shapeType=Capsule, length=6, radius=6, shapeName=head_capsule, center=<-0.2576 0.1365 0.3323>, rotation=<rot -0.6223 -80.0709 -97.0287>}]
 spine_01 [{shapeType=Capsule, length=9.005, radius=9.005, shapeName=spine_01_capsule, rotation=<rot 0 -90 0>}]
 spine_03 [{shapeType=Capsule, length=4.8535, radius=9.5365, shapeName=spine_03_capsule, center=<-2.0838 2.2176 0>}]
 spine_04 [{shapeType=Capsule, length=2.5, radius=9.8, shapeName=spine_04_capsule, center=<2.9673 3.1641 0>}]
 spine_05 [{shapeType=Capsule, length=15.5, radius=5.5, shapeName=spine_05_capsule, center=<4.6097 -1.0338 0>}]
 upperarm_l [{shapeType=Capsule, length=15.5881, radius=3.5509, shapeName=upperarm_l_capsule, center=<10.4447 0 0>, rotation=<rot -0.0001 -89.9999 -90>}]
 upperarm_r [{shapeType=Capsule, length=15.5881, radius=3.5509, shapeName=upperarm_r_capsule, center=<-10.4447 0 0>, rotation=<rot 19.9998 -89.9998 90>}]
}
constraints [
 {
  bone1Name=spine_04
  bone2Name=spine_03
  swing1LimitDegrees=45
  swing1Motion=Locked
  swing2LimitDegrees=45
  swing2Motion=Locked
  twistLimitDegrees=45
  twistMotion=Locked
 }
 {
  bone1Name=spine_05
  bone2Name=spine_04
  swing1LimitDegrees=45
  swing1Motion=Locked
  swing2LimitDegrees=45
  swing2Motion=Locked
  twistLimitDegrees=45
  twistMotion=Locked
 }
 {
  bone1Name=clavicle_l
  bone2Name=spine_05
  swing1LimitDegrees=45
  swing1Motion=Locked
  swing2LimitDegrees=45
  swing2Motion=Locked
  twistLimitDegrees=45
  twistMotion=Locked
 }
 {
  bone1Name=clavicle_r
  bone2Name=spine_05
  swing1LimitDegrees=45
  swing1Motion=Locked
  swing2LimitDegrees=45
  swing2Motion=Locked
  twistLimitDegrees=45
  twistMotion=Locked
 }
 {
  bone1Name=upperarm_r
  bone2Name=clavicle_r
  swing1LimitDegrees=45
  swing1Motion=Locked
  swing2LimitDegrees=45
  swing2Motion=Locked
  twistLimitDegrees=45
  twistMotion=Locked
 }
 {
  bone1Name=upperarm_l
  bone2Name=clavicle_l
  swing1LimitDegrees=45
  swing1Motion=Locked
  swing2LimitDegrees=45
  swing2Motion=Locked
  twistLimitDegrees=45
  twistMotion=Locked
 }
 {bone1Name=dyn_hair_l_01, bone2Name=head, swing1LimitDegrees=15, swing2LimitDegrees=10, twistLimitDegrees=5}
 {bone1Name=dyn_hair_l_02, bone2Name=dyn_hair_l_01, swing1LimitDegrees=15, swing2LimitDegrees=10, twistLimitDegrees=5}
 {bone1Name=dyn_hair_l_03, bone2Name=dyn_hair_l_02, swing1LimitDegrees=10, swing2LimitDegrees=15, twistLimitDegrees=2}
 {bone1Name=dyn_hair_l_04, bone2Name=dyn_hair_l_03, swing1LimitDegrees=10, swing2LimitDegrees=15, twistLimitDegrees=2}
 {bone1Name=dyn_hair_l_05, bone2Name=dyn_hair_l_04, swing1LimitDegrees=10, swing2LimitDegrees=15, twistLimitDegrees=2}
 {bone1Name=dyn_hair_l_06, bone2Name=dyn_hair_l_05, swing1LimitDegrees=10, swing2LimitDegrees=15, twistLimitDegrees=2}
 {bone1Name=dyn_hair_l_07, bone2Name=dyn_hair_l_06, swing1LimitDegrees=10, swing2LimitDegrees=15, twistLimitDegrees=2}
 {bone1Name=dyn_hair_r_01, bone2Name=head, swing1LimitDegrees=15, swing2LimitDegrees=10, twistLimitDegrees=5}
 {bone1Name=dyn_hair_r_02, bone2Name=dyn_hair_r_01, swing1LimitDegrees=15, swing2LimitDegrees=10, twistLimitDegrees=5}
 {bone1Name=dyn_hair_r_03, bone2Name=dyn_hair_r_02, swing1LimitDegrees=15, swing2LimitDegrees=10, twistLimitDegrees=5}
 {bone1Name=dyn_hair_r_04, bone2Name=dyn_hair_r_03, swing1LimitDegrees=15, swing2LimitDegrees=10, twistLimitDegrees=5}
 {bone1Name=dyn_hair_r_05, bone2Name=dyn_hair_r_04, swing1LimitDegrees=15, swing2LimitDegrees=10, twistLimitDegrees=5}
 {bone1Name=dyn_hair_r_06, bone2Name=dyn_hair_r_05, swing1LimitDegrees=15, swing2LimitDegrees=10, twistLimitDegrees=5}
 {bone1Name=dyn_hair_r_07, bone2Name=dyn_hair_r_06, swing1LimitDegrees=15, swing2LimitDegrees=10, twistLimitDegrees=5}
 {
  bone1Name=spine_03
  bone2Name=spine_01
  swing1LimitDegrees=45
  swing1Motion=Locked
  swing2LimitDegrees=45
  swing2Motion=Locked
  twistLimitDegrees=45
  twistMotion=Locked
 }
 {
  bone1Name=head
  bone2Name=spine_05
  swing1LimitDegrees=45
  swing1Motion=Locked
  swing2LimitDegrees=45
  swing2Motion=Locked
  twistLimitDegrees=45
  twistMotion=Locked
 }
 {
  bone1Name=dyn_hair_bk_r_02
  bone2Name=dyn_hair_bk_r_01
  swing1LimitDegrees=45
  swing1Motion=Locked
  swing2LimitDegrees=45
  swing2Motion=Locked
  twistLimitDegrees=45
  twistMotion=Locked
 }
 {
  bone1Name=dyn_hair_bk_r_01
  bone2Name=head
  swing1LimitDegrees=45
  swing1Motion=Locked
  swing2LimitDegrees=45
  swing2Motion=Locked
  twistLimitDegrees=45
  twistMotion=Locked
 }
 {
  bone1Name=dyn_hair_bk_l_02
  bone2Name=dyn_hair_bk_l_01
  swing1LimitDegrees=45
  swing1Motion=Locked
  swing2LimitDegrees=45
  swing2Motion=Locked
  twistLimitDegrees=45
  twistMotion=Locked
 }
 {
  bone1Name=dyn_hair_bk_l_01
  bone2Name=head
  swing1LimitDegrees=45
  swing1Motion=Locked
  swing2LimitDegrees=45
  swing2Motion=Locked
  twistLimitDegrees=45
  twistMotion=Locked
 }
]
