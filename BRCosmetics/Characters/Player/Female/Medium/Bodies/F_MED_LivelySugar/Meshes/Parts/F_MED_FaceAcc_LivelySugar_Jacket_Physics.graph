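# Physics asset for the LivelySugar female medium character's Jacket face-accessory part, covering two dynamic hair chains and the upper body around them.
# Left and right five-bone capsule chains hang from the head, each link constrained to its parent with tight swing and twist limits.
# Capsules and spheres on head, neck, clavicles and upper spine have no constraints and form body volumes for the hair to collide against.
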PhysicsAsset /BRCosmetics/Characters/Player/Female/Medium/Bodies/F_MED_LivelySugar/Meshes/Parts/F_MED_FaceAcc_LivelySugar_Jacket_Physics
bodies {
 clavicle_l [{
   shapeType=Capsule
   length=7.6779
   radius=4.1717
   shapeName=clavicle_l_capsule
   center=<7.2606 -0.7215 0.483>
   rotation=<rot 83.0357 -179.9977 -179.998>
  }]
 clavicle_r [{
   shapeType=Capsule
   length=4.8197
   radius=5.0637
   shapeName=clavicle_l_capsule
   center=<-6.7798 0.9816 1.3705>
   rotation=<rot 83.0357 -179.9971 0.0026>
  }]
 dyn_hair_l_01 [{shapeType=Capsule, length=0.505, radius=0.505, shapeName=dyn_hair_l_01_capsule, rotation=<rot 0 -90 0>}]
 dyn_hair_l_02 [{
   shapeType=Capsule
   length=4.7053
   radius=1.8039
   shapeName=dyn_hair_l_02_capsule
   center=<4.3135 0 0>
   rotation=<rot 9.6238 -88.2819 95.7175>
  }]
 dyn_hair_l_03 [{
   shapeType=Capsule
   length=2.6774
   radius=1.864
   shapeName=dyn_hair_l_03_capsule
   center=<3.0747 0.3454 0.0173>
   rotation=<rot -22.4139 -90.0004 89.6993>
  }]
 dyn_hair_l_04 [{
   shapeType=Capsule
   length=3.4891
   radius=2.5
   shapeName=dyn_hair_l_04_capsule
   center=<2.4046 -0.4317 -0.3378>
   rotation=<rot -14.485 -90.0003 89.5318>
  }]
 dyn_hair_l_05 [{
   shapeType=Capsule
   length=5.5447
   radius=3.8118
   shapeName=dyn_hair_l_05_capsule
   center=<7.7233 -1.2559 -1.1093>
   rotation=<rot 33.2823 100.4557 -165.0657>
  }]
 dyn_hair_r_01 [{
   shapeType=Capsule
   length=2.465
   radius=2
   shapeName=dyn_hair_r_01_capsule
   center=<4.1027 -1.3455 0.3174>
   rotation=<rot 1.5824 94.9283 -95.6467>
  }]
 dyn_hair_r_02 [{
   shapeType=Capsule
   length=3.1155
   radius=1.6738
   shapeName=dyn_hair_r_02_capsule
   center=<3.6718 -0.6639 -0.9959>
   rotation=<rot -5.8083 91.3066 -94.1698>
  }]
 dyn_hair_r_03 [{
   shapeType=Capsule
   length=3.7592
   radius=1.3932
   shapeName=dyn_hair_r_03_capsule
   center=<1.37 0.0356 -0.4471>
   rotation=<rot 26.0752 109.2074 -79.9999>
  }]
 dyn_hair_r_04 [{
   shapeType=Capsule
   length=6.0257
   radius=2.7992
   shapeName=dyn_hair_r_04_capsule
   center=<1.241 -0.245 1.495>
   rotation=<rot 22.5546 -84.6146 -96.8399>
  }]
 dyn_hair_r_05 [
  {
   shapeType=Capsule
   length=6.7521
   radius=2.2264
   shapeName=dyn_hair_r_05_capsule
   center=<8.2852 -0.33 1.0405>
   rotation=<rot 56.2676 100.0329 -96.3374>
  }
  {
   shapeType=Capsule
   length=7.0355
   radius=1.0603
   shapeName=dyn_hair_r_05_capsule
   center=<4.9609 2.5367 2.1807>
   rotation=<rot -88.762 -179.9993 -162.3495>
  }
 ]
 head [{shapeType=Capsule, length=6, radius=6, shapeName=head_capsule, center=<-0.2576 0.1365 0.3323>, rotation=<rot -0.6223 -80.0709 -97.0287>}]
 neck_01 [{shapeType=Capsule, length=4.005, radius=5.505, shapeName=neck_01_capsule, center=<-2.3276 1.507 0>, rotation=<rot 0 -90 0>}]
 spine_03 [{
   shapeType=Capsule
   length=5.8535
   radius=9.5493
   shapeName=spine_03_capsule
   center=<0.2921 2.4106 -0.4344>
   rotation=<rot 87.8773 -15.172 -19.1188>
  }]
 spine_04 [
  {shapeType=Sphere, radius=5.4387, shapeName=spine_04_sphere, center=<0.4824 7.6993 4.79>}
  {shapeType=Sphere, radius=5.8372, shapeName=spine_04_sphere, center=<0.4824 7.6993 -5.5873>}
  {shapeType=Capsule, length=10.0627, radius=7.8263, shapeName=spine_04_capsule, center=<2.9673 2.6634 0>}
  {shapeType=Capsule, length=10.5407, radius=5.2533, shapeName=spine_04_capsule, center=<9.3508 4.2887 0>}
  {shapeType=Capsule, length=11.4549, radius=7.3796, shapeName=spine_04_capsule}
 ]
 spine_05 [
  {shapeType=Capsule, length=21.5702, radius=5.2, shapeName=spine_05_capsule, center=<0.6124 -0.4864 0>}
  {shapeType=Capsule, length=20.9673, radius=9.0327, shapeName=spine_05_capsule, center=<0 9.2959 19.3071>, rotation=<rot 90 0 0>}
  {shapeType=Capsule, length=20.9673, radius=9.0327, shapeName=spine_05_capsule, center=<0 9.2959 -19.3071>, rotation=<rot 90 0 0>}
 ]
}
constraints [
 {bone1Name=dyn_hair_l_01, bone2Name=head, swing1LimitDegrees=15, swing2LimitDegrees=10, twistLimitDegrees=5}
 {bone1Name=dyn_hair_l_02, bone2Name=dyn_hair_l_01, swing1LimitDegrees=5, swing2LimitDegrees=10, twistLimitDegrees=5, twistMotion=Locked}
 {bone1Name=dyn_hair_l_03, bone2Name=dyn_hair_l_02, swing1LimitDegrees=15, swing2LimitDegrees=15, twistLimitDegrees=2}
 {bone1Name=dyn_hair_l_04, bone2Name=dyn_hair_l_03, swing1LimitDegrees=15, swing2LimitDegrees=15, twistLimitDegrees=2}
 {bone1Name=dyn_hair_l_05, bone2Name=dyn_hair_l_04, swing1LimitDegrees=20, swing2LimitDegrees=10, twistLimitDegrees=2}
 {bone1Name=dyn_hair_r_01, bone2Name=head, swing1LimitDegrees=15, swing2LimitDegrees=10, twistLimitDegrees=0}
 {bone1Name=dyn_hair_r_02, bone2Name=dyn_hair_r_01, swing1LimitDegrees=5, swing2LimitDegrees=10, twistLimitDegrees=0}
 {bone1Name=dyn_hair_r_03, bone2Name=dyn_hair_r_02, swing1LimitDegrees=15, swing2LimitDegrees=15, twistLimitDegrees=0}
 {bone1Name=dyn_hair_r_04, bone2Name=dyn_hair_r_03, swing1LimitDegrees=15, swing2LimitDegrees=15, twistLimitDegrees=0}
 {bone1Name=dyn_hair_r_05, bone2Name=dyn_hair_r_04, swing1LimitDegrees=15, swing2LimitDegrees=10, twistLimitDegrees=2}
]
