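# Physics asset for the LivelySugar female face accessory mesh, giving its dynamic hair bones collision and constrained motion.
# Five hair chains (front left and right, back center, left and right) hang from the head, each bone constrained to its parent with narrow swing and twist limits.
# Capsules on head, neck, upper spine, clavicles and upper arms, plus one spine sphere, form unconstrained body shapes around the hair, apparently as colliders.
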PhysicsAsset /BRCosmetics/Characters/Player/Female/Medium/Bodies/F_MED_LivelySugar/Meshes/Parts/F_MED_FaceAcc_LivelySugar_Physics
bodies {
 clavicle_l [{
   shapeType=Capsule
   length=7.6779
   radius=4.1717
   shapeName=clavicle_l_capsule
   center=<7.2606 -0.7215 0.483>
   rotation=<rot 83.0357 -179.9977 -179.998>
  }]
 clavicle_r [{
   shapeType=Capsule
   length=4.8197
   radius=5.0637
   shapeName=clavicle_l_capsule
   center=<-6.7798 0.9816 1.3705>
   rotation=<rot 83.0357 -179.9971 0.0026>
  }]
 dyn_hair_bk_01 [{shapeType=Capsule, length=4.1715, radius=1.9143, shapeName=dyn_hair_bk_01_capsule, center=<3 0 0>, rotation=<rot 0 90 -80>}]
 dyn_hair_bk_02 [{shapeType=Capsule, length=4.0048, radius=2.7101, shapeName=dyn_hair_bk_02_capsule, center=<3.1918 0 0>, rotation=<rot 0 90 -90>}]
 dyn_hair_bk_03 [{shapeType=Capsule, length=5, radius=3.1, shapeName=dyn_hair_bk_03_capsule, center=<4.2492 0 1.5>, rotation=<rot 0 90 -90>}]
 dyn_hair_bk_04 [{
   shapeType=Capsule
   length=2.8209
   radius=3.6723
   shapeName=dyn_hair_bk_04_capsule
   center=<3.9249 0 0>
   rotation=<rot -3.5877 89.3025 -96.5672>
  }]
 dyn_hair_bk_05 [{
   shapeType=Capsule
   length=5.2985
   radius=3.8907
   shapeName=dyn_hair_bk_05_capsule
   center=<5.1932 0.0412 -0.0318>
   rotation=<rot 0.4598 -88.3519 -91.6292>
  }]
 dyn_hair_bk_06 [{shapeType=Capsule, length=5.9639, radius=2.642, shapeName=dyn_hair_bk_06_capsule, center=<3.6047 0 0>, rotation=<rot 0 90 -90>}]
 dyn_hair_bk_l_01 [{shapeType=Capsule, length=4.6471, radius=1.7354, shapeName=dyn_hair_bk_l_01_capsule, center=<3 0 0>, rotation=<rot -0.0001 90 -80>}]
 dyn_hair_bk_l_02 [{shapeType=Capsule, length=4.0048, radius=2.7101, shapeName=dyn_hair_bk_l_02_capsule, center=<3.164 0 0.5>, rotation=<rot 0 -90 -90>}]
 dyn_hair_bk_l_03 [{shapeType=Capsule, length=5, radius=3.6, shapeName=dyn_hair_bk_l_03_capsule, center=<5.1277 0 0.5>, rotation=<rot 0 90 -90>}]
 dyn_hair_bk_l_04 [{shapeType=Capsule, length=2.8182, radius=3.6016, shapeName=dyn_hair_bk_l_04_capsule, center=<4.3429 -0.5 1>, rotation=<rot 0 90 -90>}]
 dyn_hair_bk_l_05 [{shapeType=Capsule, length=2.6027, radius=3.4653, shapeName=dyn_hair_bk_l_05_capsule, center=<3.2774 -0.6 0.4>, rotation=<rot 0 90.134 -90>}]
 dyn_hair_bk_r_01 [{shapeType=Capsule, length=4.634, radius=2.1545, shapeName=dyn_hair_bk_r_01_capsule, center=<3 0 0>, rotation=<rot 0 90 -80>}]
 dyn_hair_bk_r_02 [
  {shapeType=Capsule, length=4.0048, radius=2.7101, shapeName=dyn_hair_bk_r_02_capsule, center=<3 0 0>, rotation=<rot 0 90.9982 -85.5257>}
  {
   shapeType=Capsule
   length=4.0048
   radius=2.7101
   shapeName=dyn_hair_bk_r_02_capsule
   center=<6.7804 -2.2027 -1.5747>
   rotation=<rot 0 90.9982 -64.4899>
  }
 ]
 dyn_hair_bk_r_03 [{shapeType=Capsule, length=5.5, radius=3.6, shapeName=dyn_hair_bk_r_03_capsule, center=<4.7801 0 0.9>, rotation=<rot 0 90 -90>}]
 dyn_hair_bk_r_04 [{
   shapeType=Capsule
   length=3.1257
   radius=3.8649
   shapeName=dyn_hair_bk_r_04_capsule
   center=<3.8186 1 1>
   rotation=<rot 1.6087 90.8867 -87.2119>
  }]
 dyn_hair_bk_r_05 [{shapeType=Capsule, length=2.2956, radius=4.0192, shapeName=dyn_hair_bk_r_05_capsule, center=<3.4413 0 0>, rotation=<rot 0 90 -90>}]
 dyn_hair_l_01 [{shapeType=Capsule, length=0.505, radius=0.505, shapeName=dyn_hair_l_01_capsule, rotation=<rot 0 -90 0>}]
 dyn_hair_l_02 [
  {
   shapeType=Capsule
   length=4.7053
   radius=1.8039
   shapeName=dyn_hair_l_02_capsule
   center=<7.1576 0.3564 -1.7924>
   rotation=<rot 11.1802 -119.0101 90.0366>
  }
  {shapeType=Capsule, length=5, radius=2, shapeName=dyn_hair_l_02_capsule, rotation=<rot -66.449 180 0>}
 ]
 dyn_hair_l_03 [{
   shapeType=Capsule
   length=2.6774
   radius=1.864
   shapeName=dyn_hair_l_03_capsule
   center=<3.0747 0.3454 0.0173>
   rotation=<rot -22.4139 -90.0004 89.6993>
  }]
 dyn_hair_l_04 [{
   shapeType=Capsule
   length=3.4891
   radius=2.5
   shapeName=dyn_hair_l_04_capsule
   center=<2.4046 -0.4317 -0.3378>
   rotation=<rot -14.485 -90.0003 89.5318>
  }]
 dyn_hair_l_05 [{
   shapeType=Capsule
   length=7.1173
   radius=2.2392
   shapeName=dyn_hair_l_05_capsule
   center=<5.65 -0.42 -1.1>
   rotation=<rot -9.5674 90 -83.0668>
  }]
 dyn_hair_r_01 [{
   shapeType=Capsule
   length=2.465
   radius=2
   shapeName=dyn_hair_r_01_capsule
   center=<4.1027 -1.3455 0.3174>
   rotation=<rot 1.5824 94.9283 -95.6467>
  }]
 dyn_hair_r_02 [{
   shapeType=Capsule
   length=3.1155
   radius=1.6738
   shapeName=dyn_hair_r_02_capsule
   center=<3.6718 -0.6639 -0.9959>
   rotation=<rot -5.8083 91.3066 -94.1698>
  }]
 dyn_hair_r_03 [{
   shapeType=Capsule
   length=3.7592
   radius=1.3932
   shapeName=dyn_hair_r_03_capsule
   center=<1.37 0.0356 -0.4471>
   rotation=<rot 26.0752 109.2074 -79.9999>
  }]
 dyn_hair_r_04 [{
   shapeType=Capsule
   length=6.0257
   radius=2.7992
   shapeName=dyn_hair_r_04_capsule
   center=<1.241 -0.245 1.495>
   rotation=<rot 18.3954 -107.4107 -102.1266>
  }]
 dyn_hair_r_05 [
  {
   shapeType=Capsule
   length=6.7521
   radius=2.2264
   shapeName=dyn_hair_r_05_capsule
   center=<8.2852 -0.33 1.0405>
   rotation=<rot 56.2676 100.0329 -96.3374>
  }
  {
   shapeType=Capsule
   length=7.0355
   radius=1.0603
   shapeName=dyn_hair_r_05_capsule
   center=<4.9609 2.5367 2.1807>
   rotation=<rot -88.762 -179.9993 -162.3495>
  }
 ]
 head [{shapeType=Capsule, length=6, radius=6, shapeName=head_capsule, center=<-0.2576 0.1365 0.3323>, rotation=<rot -0.6223 -80.0709 -97.0287>}]
 neck_01 [{shapeType=Capsule, length=4.005, radius=5.505, shapeName=neck_01_capsule, center=<-2.3276 1.507 0>, rotation=<rot 0 -90 0>}]
 spine_03 [{
   shapeType=Capsule
   length=5.8535
   radius=9.5493
   shapeName=spine_03_capsule
   center=<0.2921 2.4106 -0.4344>
   rotation=<rot 87.8773 -15.172 -19.1188>
  }]
 spine_04 [
  {shapeType=Sphere, radius=5.4387, shapeName=spine_04_sphere, center=<0.4824 7.6993 4.79>}
  {shapeType=Capsule, length=10.0627, radius=7.8263, shapeName=spine_04_capsule, center=<2.9673 2.6634 0>}
  {shapeType=Capsule, length=10.5407, radius=5.2533, shapeName=spine_04_capsule, center=<9.3508 4.2887 0>}
  {shapeType=Capsule, length=11.4549, radius=7.3796, shapeName=spine_04_capsule}
 ]
 spine_05 [{shapeType=Capsule, length=21.5702, radius=5.2, shapeName=spine_05_capsule, center=<0.6124 -0.4864 0>}]
 upperarm_l [{
   shapeType=Capsule
   length=21.8786
   radius=4.3197
   shapeName=upperarm_l_capsule
   center=<10.9941 0.4228 -1.2383>
   rotation=<rot -0.0001 -89.9999 93.323>
  }]
 upperarm_r [{
   shapeType=Capsule
   length=21.4405
   radius=4.7578
   shapeName=upperarm_r_capsule
   center=<-10.9941 -0.4228 1.2383>
   rotation=<rot -0.0001 -89.9999 -86.677>
  }]
}
constraints [
 {bone1Name=dyn_hair_l_01, bone2Name=head, swing1LimitDegrees=15, swing2LimitDegrees=10, twistLimitDegrees=5}
 {bone1Name=dyn_hair_l_02, bone2Name=dyn_hair_l_01, swing1LimitDegrees=5, swing2LimitDegrees=10, twistLimitDegrees=5, twistMotion=Locked}
 {bone1Name=dyn_hair_l_03, bone2Name=dyn_hair_l_02, swing1LimitDegrees=10, swing2LimitDegrees=15, twistLimitDegrees=2}
 {bone1Name=dyn_hair_l_04, bone2Name=dyn_hair_l_03, swing1LimitDegrees=10, swing2LimitDegrees=15, twistLimitDegrees=2}
 {bone1Name=dyn_hair_l_05, bone2Name=dyn_hair_l_04, swing1LimitDegrees=20, swing2LimitDegrees=10, twistLimitDegrees=2}
 {bone1Name=dyn_hair_bk_l_01, bone2Name=head, swing1LimitDegrees=6, swing2LimitDegrees=13, twistLimitDegrees=2}
 {bone1Name=dyn_hair_bk_l_02, bone2Name=dyn_hair_bk_l_01, swing1LimitDegrees=5, swing2LimitDegrees=20, twistLimitDegrees=2}
 {bone1Name=dyn_hair_bk_l_03, bone2Name=dyn_hair_bk_l_02, swing1LimitDegrees=10, swing2LimitDegrees=20, twistLimitDegrees=2}
 {bone1Name=dyn_hair_bk_l_04, bone2Name=dyn_hair_bk_l_03, swing1LimitDegrees=10, swing2LimitDegrees=20, twistLimitDegrees=2}
 {bone1Name=dyn_hair_bk_l_05, bone2Name=dyn_hair_bk_l_04, swing1LimitDegrees=10, swing2LimitDegrees=20, twistLimitDegrees=2}
 {bone1Name=dyn_hair_bk_01, bone2Name=head, swing1LimitDegrees=6, swing2LimitDegrees=13, twistLimitDegrees=2}
 {bone1Name=dyn_hair_bk_02, bone2Name=dyn_hair_bk_01, swing1LimitDegrees=5, swing2LimitDegrees=30, twistLimitDegrees=2}
 {bone1Name=dyn_hair_bk_03, bone2Name=dyn_hair_bk_02, swing1LimitDegrees=5, swing2LimitDegrees=15, twistLimitDegrees=2}
 {bone1Name=dyn_hair_bk_04, bone2Name=dyn_hair_bk_03, swing1LimitDegrees=5, swing2LimitDegrees=30, twistLimitDegrees=2}
 {bone1Name=dyn_hair_bk_05, bone2Name=dyn_hair_bk_04, swing1LimitDegrees=5, swing2LimitDegrees=30, twistLimitDegrees=2}
 {bone1Name=dyn_hair_bk_06, bone2Name=dyn_hair_bk_05, swing1LimitDegrees=5, swing2LimitDegrees=20, twistLimitDegrees=2}
 {bone1Name=dyn_hair_bk_r_01, bone2Name=head, swing1LimitDegrees=6, swing2LimitDegrees=13, twistLimitDegrees=2}
 {bone1Name=dyn_hair_bk_r_02, bone2Name=dyn_hair_bk_r_01, swing1LimitDegrees=0, swing2LimitDegrees=20, twistLimitDegrees=2}
 {bone1Name=dyn_hair_bk_r_03, bone2Name=dyn_hair_bk_r_02, swing1LimitDegrees=10, swing2LimitDegrees=20, twistLimitDegrees=0}
 {bone1Name=dyn_hair_bk_r_04, bone2Name=dyn_hair_bk_r_03, swing1LimitDegrees=10, swing2LimitDegrees=20, twistLimitDegrees=0}
 {bone1Name=dyn_hair_bk_r_05, bone2Name=dyn_hair_bk_r_04, swing1LimitDegrees=10, swing2LimitDegrees=20, twistLimitDegrees=0}
 {bone1Name=dyn_hair_r_01, bone2Name=head, swing1LimitDegrees=15, swing2LimitDegrees=10, twistLimitDegrees=0}
 {bone1Name=dyn_hair_r_02, bone2Name=dyn_hair_r_01, swing1LimitDegrees=5, swing2LimitDegrees=10, twistLimitDegrees=0}
 {bone1Name=dyn_hair_r_03, bone2Name=dyn_hair_r_02, swing1LimitDegrees=15, swing2LimitDegrees=10, twistLimitDegrees=0}
 {bone1Name=dyn_hair_r_04, bone2Name=dyn_hair_r_03, swing1LimitDegrees=15, swing2LimitDegrees=10, twistLimitDegrees=0}
 {bone1Name=dyn_hair_r_05, bone2Name=dyn_hair_r_04, swing1LimitDegrees=15, swing2LimitDegrees=10, twistLimitDegrees=2}
]
