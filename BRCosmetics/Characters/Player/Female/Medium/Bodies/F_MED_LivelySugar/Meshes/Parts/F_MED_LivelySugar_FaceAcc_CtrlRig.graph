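# Control Rig for the LivelySugar female face-accessory part that procedurally corrects long back-hair and side-hair bones as the body moves.
# Parent constraints blend hair control nulls between head, spine, clavicles and pelvis, then twist functions and B-spline IK shape left and right chains, whose averaged pose drives most centre bones.
# Spherical pose readers and head and clavicle rotations weight corrective offsets, some gated or blended by skydive, backpack, frontend, emote and crouch variables.
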ControlRigBlueprint /BRCosmetics/Characters/Player/Female/Medium/Bodies/F_MED_LivelySugar/Meshes/Parts/F_MED_LivelySugar_FaceAcc_CtrlRig
graphs [
 CollapseNode_ContainedGraph {nodes [
   n1 = RigVMFunctionEntryNode()
   &RigUnit_ParentConstraint(
    AdvancedSettings=(RotationOrderForFilter=XZY)
    Child=(Type=Null,Name="dyn_hair_bk_l_01_ctrl_Null")
    Filter=(TranslationFilter=(bX=True,bY=True,bZ=True),RotationFilter=(bX=True,bY=True,bZ=True),ScaleFilter=(bX=False,bY=False,bZ=False))
    Parents=((Item=(Type=Bone,Name="head"),Weight=0.85),(Item=(Type=Bone,Name="spine_05"),Weight=0.15))
    Weight=1
    ExecutePin=@n1.ExecutePin
   )
   > &RigUnit_ParentConstraint(
    AdvancedSettings=(RotationOrderForFilter=XZY)
    Child=(Type=Null,Name="dyn_hair_bk_r_01_ctrl_Null")
    Filter=(TranslationFilter=(bX=True,bY=True,bZ=True),RotationFilter=(bX=True,bY=True,bZ=True),ScaleFilter=(bX=False,bY=False,bZ=False))
    Parents=((Item=(Type=Bone,Name="head"),Weight=0.85),(Item=(Type=Bone,Name="spine_05"),Weight=0.15))
    Weight=1
   )
   > &RigUnit_ParentConstraint(
    AdvancedSettings=(RotationOrderForFilter=XZY)
    Child=(Type=Null,Name="dyn_hair_bk_l_02_ctrl_Null")
    Filter=(TranslationFilter=(bX=True,bY=True,bZ=True),RotationFilter=(bX=True,bY=True,bZ=True),ScaleFilter=(bX=False,bY=False,bZ=False))
    Parents=((Item=(Type=Bone,Name="head"),Weight=0.35),(Item=(Type=Bone,Name="spine_05"),Weight=0.65))
    Weight=1
   )
   > &RigUnit_ParentConstraint(
    AdvancedSettings=(RotationOrderForFilter=XZY)
    Child=(Type=Null,Name="dyn_hair_bk_r_02_ctrl_Null")
    Filter=(TranslationFilter=(bX=True,bY=True,bZ=True),RotationFilter=(bX=True,bY=True,bZ=True),ScaleFilter=(bX=False,bY=False,bZ=False))
    Parents=((Item=(Type=Bone,Name="head"),Weight=0.35),(Item=(Type=Bone,Name="spine_05"),Weight=0.65))
    Weight=1
   )
   > &RigUnit_ParentConstraint(
    AdvancedSettings=(RotationOrderForFilter=XZY)
    Child=(Type=Null,Name="dyn_hair_bk_l_03_ctrl_Null")
    Filter=(TranslationFilter=(bX=True,bY=True,bZ=True),RotationFilter=(bX=True,bY=True,bZ=True),ScaleFilter=(bX=False,bY=False,bZ=False))
    Parents=((Item=(Type=Bone,Name="spine_05"),Weight=0.35),(Item=(Type=Bone,Name="clavicle_l"),Weight=0.65))
    Weight=1
   )
   > &RigUnit_ParentConstraint(
    AdvancedSettings=(RotationOrderForFilter=XZY)
    Child=(Type=Null,Name="dyn_hair_bk_r_03_ctrl_Null")
    Filter=(TranslationFilter=(bX=True,bY=True,bZ=True),RotationFilter=(bX=True,bY=True,bZ=True),ScaleFilter=(bX=False,bY=False,bZ=False))
    Parents=((Item=(Type=Bone,Name="spine_05"),Weight=0.35),(Item=(Type=Bone,Name="clavicle_r"),Weight=0.65))
    Weight=1
   )
   > &RigUnit_ParentConstraint(
    AdvancedSettings=(RotationOrderForFilter=XZY)
    Child=(Type=Null,Name="dyn_hair_bk_l_04_ctrl_Null")
    Filter=(TranslationFilter=(bX=True,bY=True,bZ=True),RotationFilter=(bX=True,bY=True,bZ=True),ScaleFilter=(bX=False,bY=False,bZ=False))
    Parents=((Item=(Type=Bone,Name="spine_02"),Weight=0.0014),(Item=(Type=Bone,Name="spine_03"),Weight=0.0235),(Item=(Type=Bone,Name="spine_04"),Weight=0.0379),(Item=(Type=Bone,Name="spine_05"),Weight=0.3523),(Item=(Type=Bone,Name="clavicle_l"),Weight=0.5116))
    Weight=1
   )
   > &RigUnit_ParentConstraint(
    AdvancedSettings=(RotationOrderForFilter=XZY)
    Child=(Type=Null,Name="dyn_hair_bk_04_ctrl_Null")
    Filter=(TranslationFilter=(bX=True,bY=True,bZ=True),RotationFilter=(bX=True,bY=True,bZ=True),ScaleFilter=(bX=False,bY=False,bZ=False))
    Parents="((Item=(Type=Bone,Name=\"spine_01\"),Weight=0.0001),(Item=(Type=Bone,Name=\"spine_02\"),Weight=0.0068),(Item=(Type=Bone,Name=\"clavicle_l\"),Weight=0.063),(Item=(Type=Bone,Name=\"clavicle_r\"),Weight=0.0705),(Item=(Type=Bone,Name=\"spine_03\"),Weight=0.1021),(Item=( ...(truncated)"
    Weight=1
   )
   > &RigUnit_ParentConstraint(
    AdvancedSettings=(RotationOrderForFilter=XZY)
    Child=(Type=Null,Name="dyn_hair_bk_r_04_ctrl_Null")
    Filter=(TranslationFilter=(bX=True,bY=True,bZ=True),RotationFilter=(bX=True,bY=True,bZ=True),ScaleFilter=(bX=False,bY=False,bZ=False))
    Parents=((Item=(Type=Bone,Name="spine_02"),Weight=0.0014),(Item=(Type=Bone,Name="spine_03"),Weight=0.0235),(Item=(Type=Bone,Name="spine_04"),Weight=0.038),(Item=(Type=Bone,Name="spine_05"),Weight=0.3528),(Item=(Type=Bone,Name="clavicle_r"),Weight=0.5109))
    Weight=1
   )
   > &RigUnit_ParentConstraint(
    AdvancedSettings=(RotationOrderForFilter=XZY)
    Child=(Type=Null,Name="dyn_hair_bk_l_05_ctrl_Null")
    Filter=(TranslationFilter=(bX=True,bY=True,bZ=True),RotationFilter=(bX=True,bY=True,bZ=True),ScaleFilter=(bX=False,bY=False,bZ=False))
    Parents=((Item=(Type=Bone,Name="clavicle_l"),Weight=0.1),(Item=(Type=Bone,Name="spine_04"),Weight=0.15),(Item=(Type=Bone,Name="spine_05"),Weight=0.15),(Item=(Type=Bone,Name="spine_02"),Weight=0.2),(Item=(Type=Bone,Name="spine_03"),Weight=0.4))
    Weight=1
   )
   > &RigUnit_ParentConstraint(
    AdvancedSettings=(RotationOrderForFilter=XZY)
    Child=(Type=Null,Name="dyn_hair_bk_05_ctrl_Null")
    Filter=(TranslationFilter=(bX=True,bY=True,bZ=True),RotationFilter=(bX=True,bY=True,bZ=True),ScaleFilter=(bX=False,bY=False,bZ=False))
    Parents=((Item=(Type=Bone,Name="spine_04"),Weight=0.15),(Item=(Type=Bone,Name="spine_05"),Weight=0.1),(Item=(Type=Bone,Name="spine_02"),Weight=0.25),(Item=(Type=Bone,Name="spine_03"),Weight=0.5))
    Weight=1
   )
   > &RigUnit_ParentConstraint(
    AdvancedSettings=(RotationOrderForFilter=XZY)
    Child=(Type=Null,Name="dyn_hair_bk_r_05_ctrl_Null")
    Filter=(TranslationFilter=(bX=True,bY=True,bZ=True),RotationFilter=(bX=True,bY=True,bZ=True),ScaleFilter=(bX=False,bY=False,bZ=False))
    Parents=((Item=(Type=Bone,Name="clavicle_r"),Weight=0.1),(Item=(Type=Bone,Name="spine_04"),Weight=0.15),(Item=(Type=Bone,Name="spine_05"),Weight=0.15),(Item=(Type=Bone,Name="spine_02"),Weight=0.2),(Item=(Type=Bone,Name="spine_03"),Weight=0.4))
    Weight=1
   )
   > &RigUnit_ParentConstraint(
    AdvancedSettings=(RotationOrderForFilter=XZY)
    Child=(Type=Null,Name="dyn_hair_bk_l_06_ctrl_Null")
    Filter=(TranslationFilter=(bX=True,bY=True,bZ=True),RotationFilter=(bX=True,bY=True,bZ=True),ScaleFilter=(bX=False,bY=False,bZ=False))
    Parents=((Item=(Type=Bone,Name="spine_02"),Weight=0.075),(Item=(Type=Bone,Name="spine_01"),Weight=0.425),(Item=(Type=Bone,Name="pelvis"),Weight=0.5))
    Weight=1
   )
   > &RigUnit_ParentConstraint(
    AdvancedSettings=(RotationOrderForFilter=XZY)
    Child=(Type=Null,Name="dyn_hair_bk_06_ctrl_Null")
    Filter=(TranslationFilter=(bX=True,bY=True,bZ=True),RotationFilter=(bX=True,bY=True,bZ=True),ScaleFilter=(bX=False,bY=False,bZ=False))
    Parents=((Item=(Type=Bone,Name="spine_02"),Weight=0.075),(Item=(Type=Bone,Name="spine_01"),Weight=0.425),(Item=(Type=Bone,Name="pelvis"),Weight=0.5))
    Weight=1
   )
   > &RigUnit_ParentConstraint(
    AdvancedSettings=(RotationOrderForFilter=XZY)
    Child=(Type=Null,Name="dyn_hair_bk_r_06_ctrl_Null")
    Filter=(TranslationFilter=(bX=True,bY=True,bZ=True),RotationFilter=(bX=True,bY=True,bZ=True),ScaleFilter=(bX=False,bY=False,bZ=False))
    Parents=((Item=(Type=Bone,Name="spine_02"),Weight=0.075),(Item=(Type=Bone,Name="spine_01"),Weight=0.425),(Item=(Type=Bone,Name="pelvis"),Weight=0.5))
    Weight=1
   )
   > n2 = &RigUnit_ParentConstraint(
    AdvancedSettings=(RotationOrderForFilter=XZY)
    Child=(Type=Null,Name="dyn_hair_bk_07_ctrl_Null")
    Filter=(TranslationFilter=(bX=True,bY=True,bZ=True),RotationFilter=(bX=True,bY=True,bZ=True),ScaleFilter=(bX=False,bY=False,bZ=False))
    Parents=((Item=(Type=Bone,Name="spine_01"),Weight=0.02),(Item=(Type=Bone,Name="pelvis"),Weight=0.98))
    Weight=1
   )
   RigVMFunctionReturnNode(ExecutePin=@n2.ExecutePin)
   &RigUnit_ParentConstraint(
    AdvancedSettings=(RotationOrderForFilter=XZY)
    Child=(Type=Null,Name="dyn_hair_bk_01_ctrl_Null")
    ExecutePin=()
    Filter=(TranslationFilter=(bX=True,bY=True,bZ=True),RotationFilter=(bX=True,bY=True,bZ=True),ScaleFilter=(bX=False,bY=False,bZ=False))
    Parents=((Item=(Type=Bone,Name="head"),Weight=0.85),(Item=(Type=Bone,Name="spine_05"),Weight=0.15))
    Weight=1
   )
   &RigUnit_ParentConstraint(
    AdvancedSettings=(RotationOrderForFilter=XZY)
    Child=(Type=Null,Name="dyn_hair_bk_02_ctrl_Null")
    ExecutePin=()
    Filter=(TranslationFilter=(bX=True,bY=True,bZ=True),RotationFilter=(bX=True,bY=True,bZ=True),ScaleFilter=(bX=False,bY=False,bZ=False))
    Parents=((Item=(Type=Bone,Name="head"),Weight=0.35),(Item=(Type=Bone,Name="spine_05"),Weight=0.65))
    Weight=1
   )
   &RigUnit_ParentConstraint(
    AdvancedSettings=(RotationOrderForFilter=XZY)
    Child=(Type=Null,Name="dyn_hair_bk_03_ctrl_Null")
    ExecutePin=()
    Filter=(TranslationFilter=(bX=True,bY=True,bZ=True),RotationFilter=(bX=True,bY=True,bZ=True),ScaleFilter=(bX=False,bY=False,bZ=False))
    Parents=((Item=(Type=Bone,Name="clavicle_l"),Weight=0.05),(Item=(Type=Bone,Name="clavicle_r"),Weight=0.05),(Item=(Type=Bone,Name="spine_05"),Weight=0.9))
    Weight=1
   )
  ]}
 Multiply_3_ContainedGraph {nodes [
   n1 = RigVMFunctionEntryNode()
   RigVMFunctionReturnNode(Result=&RigVMFunction_MathDoubleMul(B=@n1.C, A=&RigVMFunction_MathDoubleMul(A=@n1.A, B=@n1.B).Result).Result)
  ]}
 New Function_ContainedGraph {nodes [
   n1 = RigVMFunctionEntryNode()
   RigVMFunction_ControlFlowBranch = &RigVMFunction_ControlFlowBranch(ExecuteContext=@n1.ExecuteContext, Condition=@n1.debug)
   RigVMFunctionReturnNode(ExecuteContext=@RigVMFunction_ControlFlowBranch.Completed)
   ItemArray = RigVMDispatchNode(Value=((Type=Bone,Name="head"))) {template="DISPATCH_RigVMDispatch_Constant(io Value)"}
   For_Each = RigVMDispatchNode(Array=@ItemArray.Value, ExecuteContext=@RigVMFunction_ControlFlowBranch.True) {template="DISPATCH_RigVMDispatch_ArrayIterator(in Array,out Element,out Index,out Count,out Ratio)"}
   Concat = /Script/RigVM.RigVMFunction_NameConcat(B=_null, A=@For_Each.Element.Name) {is=struct RigVMUnitNode, template="Concat::Execute(in A,in B,out Result)"}
   RigUnit_GetTransform = &RigUnit_GetTransform(Item=(Type=Bone,Name="None"), Space=LocalSpace, bInitial=true, Item.Name=@For_Each.Element.Name)
   > &RigUnit_HierarchyAddNull(
    Space=LocalSpace
    Name=@Concat.Result
    Parent=/Script/ControlRig.RigUnit_HierarchyGetParent(bDefaultParent=True, Child=@For_Each.Element) {is=struct RigVMUnitNode}.Parent
    Transform=@RigUnit_GetTransform.Transform
   )
   &RigUnit_HierarchyAddNull(
    Name=head_sp_null
    Parent=(Type=Bone,Name="spine_05")
    Space=LocalSpace
    ExecutePin=@For_Each.Completed
    Transform=&RigUnit_GetRelativeTransformForItem(Child=(Type=Bone,Name="head"), Parent=(Type=Bone,Name="spine_05"), bChildInitial=true, bParentInitial=true).RelativeTransform
   )
  ]}
 New Function_ContainedGraph {nodes [
   n1 = RigVMFunctionEntryNode()
   RigVMFunction_MathQuaternionToEuler = &RigVMFunction_MathQuaternionToEuler(
    RotationOrder=XYZ
    Value=&RigUnit_GetRelativeTransformForItem(
     Child=(Type=Bone,Name="hand_l")
     Parent=(Type=Null,Name="lowerarm_l")
     bChildInitial=False
     bParentInitial=False
     Child.Name=@n1.Child
     Parent.Name=@n1.Parent
    ).RelativeTransform.Rotation
   )
   Remap = &RigVMFunction_MathDoubleRemap(
    TargetMaximum=1
    TargetMinimum=0
    Value=@RigVMFunction_MathQuaternionToEuler.Result.X
    SourceMaximum=@n1.Source Maximum
    SourceMinimum=@n1.Source Minimum
   )
   > RigVMFunctionReturnNode(Delta Rotation=@Remap.Result)
  ]}
 New Function_ContainedGraph {nodes [
   n1 = RigVMFunctionEntryNode()
   RigVMFunction_MathQuaternionToEuler = &RigVMFunction_MathQuaternionToEuler(
    RotationOrder=XYZ
    Value=&RigUnit_GetRelativeTransformForItem(
     Child=(Type=Bone,Name="hand_l")
     Parent=(Type=Null,Name="lowerarm_l")
     bChildInitial=False
     bParentInitial=False
     Child.Name=@n1.Child
     Parent.Name=@n1.Parent
    ).RelativeTransform.Rotation
   )
   Remap = &RigVMFunction_MathDoubleRemap(
    TargetMaximum=1
    TargetMinimum=0
    Value=@RigVMFunction_MathQuaternionToEuler.Result.Z
    SourceMaximum=@n1.Source Maximum
    SourceMinimum=@n1.Source Minimum
   )
   > RigVMFunctionReturnNode(Delta Rotation=@Remap.Result)
  ]}
 New Function_ContainedGraph {nodes [
   n1 = RigVMFunctionEntryNode()
   RigVMFunction_MathBoolToInteger = &RigVMFunction_MathBoolToInteger(Value=@n1.onlyTwist)
   RigVMFunction_MathQuaternionSwingTwist = &RigVMFunction_MathQuaternionSwingTwist(
    TwistAxis=@n1.twistAxis
    Input=&RigUnit_GetRelativeTransformForItem(bChildInitial=False, bParentInitial=False, Child=@n1.child, Parent=@n1.parent).RelativeTransform.Rotation
   )
   n2 = RigVMDispatchNode(
    Values=((),())
    Index=@RigVMFunction_MathBoolToInteger.Result
    Values.0=@RigVMFunction_MathQuaternionSwingTwist.Swing
    Values.1=@RigVMFunction_MathQuaternionSwingTwist.Twist
   ) {template="DISPATCH_RigVMDispatch_SelectInt32(in Index,in Values,out Result)"}
   RigVMFunction_MathQuaternionToEuler = &RigVMFunction_MathQuaternionToEuler(Value=@n2.Result, RotationOrder=@n1.Rotation Order)
   > DISPATCH_RigVMDispatch_Print = RigVMDispatchNode(ScreenColor=(R=1,G=1,B=1,A=1), ScreenDuration=1.5, Enabled=@n1.debugRotation, Value=@RigVMFunction_MathQuaternionToEuler.Result) {template="DISPATCH_RigVMDispatch_Print(in Prefix,in Value,in Enabled,in ScreenDuration,in ScreenColor)"}
   Cast = RigVMDispatchNode(Value=@n1.triggerAxis) {template="DISPATCH_RigVMDispatch_CastEnumToInt(in Value,out Result)"}
   DISPATCH_RigVMDispatch_SelectInt32 = RigVMDispatchNode(
    Values=(0,0,0)
    Index=@Cast.Result
    Values.0=@RigVMFunction_MathQuaternionToEuler.Result.X
    Values.1=@RigVMFunction_MathQuaternionToEuler.Result.Y
    Values.2=@RigVMFunction_MathQuaternionToEuler.Result.Z
   ) {template="DISPATCH_RigVMDispatch_SelectInt32(in Index,in Values,out Result)"}
   Remap = &RigVMFunction_MathDoubleRemap(
    bClamp=False
    Value=@DISPATCH_RigVMDispatch_SelectInt32.Result
    TargetMinimum=@n1.outputMinMax.X
    TargetMaximum=@n1.outputMinMax.Y
    SourceMinimum=@n1.rotationMinMax.X
    SourceMaximum=@n1.rotationMinMax.Y
   )
   Absolute = &RigVMFunction_MathDoubleAbs(Value=@Remap.Result)
   Multiply = &RigVMFunction_MathDoubleMul(
    A=&RigVMFunction_MathDoubleClamp(Value=RigVMDispatchNode(True=@Absolute.Result, Condition=@n1.absoluteValue, False=@Remap.Result) {template="DISPATCH_RigVMDispatch_If(in Condition,in True,in False,out Result)"}.Result).Result
    B=@n1.premult
   )
   > RigVMFunctionReturnNode(outputQuat=@n2.Result, outputWeight=@Multiply.Result)
  ]}
 New Function_ContainedGraph {nodes [
   n1 = RigVMFunctionEntryNode()
   > DISPATCH_RigVMDispatch_ArrayIterator = RigVMDispatchNode(Array=@n1.transforms) {template="DISPATCH_RigVMDispatch_ArrayIterator(in Array,out Element,out Index,out Count,out Ratio)"}
   Multiply = &RigVMFunction_MathTransformMul(
    A=@DISPATCH_RigVMDispatch_ArrayIterator.Element.Transform
    B=&RigUnit_GetTransform(
     Item=(Type=Bone,Name="dyn_L_hair_3")
     Space=LocalSpace
     bInitial=False
     Item.Name=@DISPATCH_RigVMDispatch_ArrayIterator.Element.Bone
     Item.Type=@n1.type
    ).Transform
   ) {template="Multiply::Execute(in A,in B,out Result)"}
   > &RigUnit_SetTransform(
    Item=(Type=Bone,Name="dyn_L_hair_3")
    Space=LocalSpace
    bInitial=false
    Value=@Multiply.Result
    bPropagateToChildren=@n1.inheritsTransform
    Weight=@n1.weight
    Item.Name=@DISPATCH_RigVMDispatch_ArrayIterator.Element.Bone
    Item.Type=@n1.type
   )
   RigVMFunctionReturnNode(ExecuteContext=@DISPATCH_RigVMDispatch_ArrayIterator.Completed)
   &RigUnit_OffsetTransformForItem(
    ExecutePin=()
    Item=(Type=Bone,Name="None")
    OffsetTransform=(Rotation=(X=0,Y=0,Z=0,W=1),Translation=(X=0,Y=0,Z=0),Scale3D=(X=1,Y=1,Z=1))
    Weight=1
    bPropagateToChildren=True
   )
  ]}
 New Function_ContainedGraph {nodes [
   n1 = RigVMFunctionEntryNode()
   RigVMFunction_MathQuaternionToEuler = &RigVMFunction_MathQuaternionToEuler(
    RotationOrder=XYZ
    Value=&RigUnit_GetRelativeTransformForItem(
     Child=(Type=Bone,Name="hand_l")
     Parent=(Type=Null,Name="lowerarm_l")
     bChildInitial=False
     bParentInitial=False
     Child.Name=@n1.Child
     Parent.Name=@n1.Parent
    ).RelativeTransform.Rotation
   )
   Remap = &RigVMFunction_MathDoubleRemap(
    TargetMaximum=1
    TargetMinimum=0
    Value=@RigVMFunction_MathQuaternionToEuler.Result.Y
    SourceMaximum=@n1.Source Maximum
    SourceMinimum=@n1.Source Minimum
   )
   > RigVMFunctionReturnNode(Delta Rotation=@Remap.Result)
  ]}
 New Function_ContainedGraph {nodes [
   n1 = RigVMFunctionEntryNode()
   RigVMFunction_Sequence = &RigVMFunction_Sequence(ExecuteContext=@n1.ExecuteContext)
   Delta Rotation X_1 = RigVMFunctionReferenceNode(Child=head, Parent=head_null, Source Maximum=-75, ExecuteContext=@RigVMFunction_Sequence.A)
   > &RigUnit_ModifyTransforms(
    ItemToModify="((Item=(Type=Bone,Name=\"base_hair_l_01\"),Transform=(Rotation=(X=0,Y=0,Z=0,W=1),Translation=(X=0,Y=0,Z=-3.0318),Scale3D=(X=1,Y=1,Z=1))),(Item=(Type=Bone,Name=\"base_hair_bk_l_01\"),Transform=(Rotation=(X=0,Y=0,Z=0,W=1),Translation=(X=2.2919,Y=-2.8476,Z=1.4621 ...(truncated)"
    WeightMinimum=0
    Weight=@"Delta Rotation X_1".Delta Rotation
   )
   > Delta Rotation X_2 = RigVMFunctionReferenceNode(Child=head, Parent=head_null, Source Maximum=67, Source Minimum=30)
   > &RigUnit_ModifyTransforms(
    ItemToModify="((Item=(Type=Bone,Name=\"base_hair_bk_r_01\"),Transform=(Rotation=(X=0,Y=0,Z=0,W=1),Translation=(X=1.5326,Y=-2.7901,Z=1.158),Scale3D=(X=1,Y=1,Z=1))),(Item=(Type=Bone,Name=\"base_hair_r_01\"),Transform=(Rotation=(X=0,Y=0,Z=0,W=1),Translation=(X=0.0932,Y=-1.0061 ...(truncated)"
    WeightMinimum=0
    Weight=@"Delta Rotation X_2".Delta Rotation
   )
   RigVMFunctionReturnNode(ExecuteContext=@RigVMFunction_Sequence.B)
  ]}
 New Function_ContainedGraph {nodes [
   RigVMFunctionEntryNode()
   > RigVMFunctionReturnNode()
  ]}
 RigVMFunctionLibrary {nodes [
   Delta Rotation X = RigVMCollapseNode(ExecuteContext=())
   Delta Rotation Y = RigVMCollapseNode(ExecuteContext=())
   Delta Rotation Z = RigVMCollapseNode(ExecuteContext=())
   DeltaRotation v4 = RigVMCollapseNode(
    ExecuteContext=()
    Rotation Order=ZYX
    child=(Type=Bone,Name="None")
    outputMinMax=(X=0,Y=1)
    parent=(Type=Bone,Name="None")
    premult=1
    rotationMinMax=(X=0,Y=0)
    twistAxis=(X=1,Y=0,Z=0)
   )
   HeadAdjust = RigVMCollapseNode(ExecuteContext=())
   New Function = RigVMCollapseNode(ExecuteContext=())
   alterTransform = RigVMCollapseNode(type=Bone, ExecuteContext=(), transforms=())
   controllers = RigVMCollapseNode(ExecuteContext=(), debug=False)
  ]}
 RigVMFunction_Sequence_1_ContainedGraph {nodes [
   n1 = RigVMFunctionEntryNode()
   RigVMFunction_Sequence = &RigVMFunction_Sequence(ExecuteContext=@n1.ExecuteContext)
   n2 = &RigVMFunction_Sequence(ExecuteContext=@RigVMFunction_Sequence.B)
   n3 = &RigVMFunction_Sequence(ExecuteContext=@n2.B)
   RigVMFunctionReturnNode(A=@RigVMFunction_Sequence.A, B=@n2.A, C=@n3.A, D=@n3.B)
  ]}
 RigVMModel {nodes [
   n1 = &RigUnit_PrepareForExecution()
   addFKCtrl = RigVMFunctionReferenceNode(
    Custom_FKCtrl=()
    Options_FKCtrl="(pelvis_1_#1=False,spine_3_#2=true,clavicle_5_#3=true,arm_7_#4=true,leg_9_#5=False,head_11_46DC15EF465CD29 ...(truncated)"
    Shape_Settings="(pelvis_2_#6=(bVisible=True,Name=\"Hexagon_Thin\",Color=(R=0.0117,G=0.8281,B=1,A=1),Transform=(Rotation=(X=0,Y=-0.7071,Z=0,W=0.7071),Translation=(X=2,Y=0,Z=0),Scale3D=(X=5,Y=5,Z=7))),spine_4_#7=(bVi ...(truncated)"
    ExecuteContext=@n1.ExecutePin
   )
   RigUnit_BeginExecution = &RigUnit_BeginExecution()
   setFKCtrl = RigVMFunctionReferenceNode(setFK_Ctrl=false, ExecuteContext=@RigUnit_BeginExecution.ExecutePin)
   Constraints = RigVMCollapseNode(ExecutePin=@setFKCtrl.ExecuteContext)
   RigVMFunction_Sequence = RigVMAggregateNode(ExecuteContext=@Constraints.ExecutePin)
   ItemArray_7 = RigVMDispatchNode(Value=((Type=Bone,Name="None"),(Type=Bone,Name="dyn_hair_bk_l_03"),(Type=Bone,Name="dyn_hair_bk_l_04"),(Type=Bone,Name="dyn_hair_bk_l_05"),(Type=Bone,Name="dyn_hair_bk_l_06"))) {template="DISPATCH_RigVMDispatch_Constant(io Value)"}
   TwistBones_1 = RigVMFunctionReferenceNode(
    Axis to Twist=(bX=True,bY=False,bZ=False)
    End=(Type=Bone,Name="neck_01")
    Start=(Type=Bone,Name="spine_02")
    Bones to Twist=@ItemArray_7.Value
    ExecuteContext=@RigVMFunction_Sequence.A
   )
   ItemArray_8 = RigVMDispatchNode(Value=((Type=Bone,Name="dyn_hair_bk_r_02"),(Type=Bone,Name="dyn_hair_bk_r_03"),(Type=Bone,Name="dyn_hair_bk_r_04"),(Type=Bone,Name="dyn_hair_bk_r_05"),(Type=Bone,Name="dyn_hair_bk_r_06"))) {template="DISPATCH_RigVMDispatch_Constant(io Value)"}
   > TwistBones_2 = RigVMFunctionReferenceNode(
    Axis to Twist=(bX=True,bY=False,bZ=False)
    End=(Type=Bone,Name="neck_01")
    Start=(Type=Bone,Name="spine_02")
    Bones to Twist=@ItemArray_8.Value
   )
   ItemArray = RigVMDispatchNode(Value=((Type=Bone,Name="base_hair_bk_l_01"),(Type=Bone,Name="dyn_hair_bk_l_01"),(Type=Bone,Name="dyn_hair_bk_l_02"),(Type=Bone,Name="dyn_hair_bk_l_03"),(Type=Bone,Name="dyn_hair_bk_l_04"),(Type=Bone,Name="dyn_hair_bk_l_05"),(Type=Bone,Name="dyn_hair_bk_l_06"))) {template="DISPATCH_RigVMDispatch_Constant(io Value)"}
   ItemArray_1 = RigVMDispatchNode(Value="((Type=Control,Name=\"dyn_hair_bk_l_01_ctrl\"),(Type=Control,Name=\"dyn_hair_bk_l_02_ctrl\"),(Type=Control,Name=\"dyn_hair_bk_l_03_ctrl\"),(Type=Control,Name=\"dyn_hair_bk_l_04_ctrl\"),(Type=Control,Name=\"dyn_hair_bk_l_05_ctrl\"),(Type=Control,Name=\"dyn_hair_bk_l_0 ...(truncated)") {template="DISPATCH_RigVMDispatch_Constant(io Value)"}
   > SplineIK = RigVMFunctionReferenceNode(
    Primary Axis=(X=1,Y=0,Z=0)
    Samples Per Segment=16
    Secondary Spline Direction=(X=0,Y=5,Z=0)
    Spline Mode=BSpline
    Up Axis=(X=0,Y=1,Z=0)
    Bones=@ItemArray.Value
    Controls=@ItemArray_1.Value
   )
   ItemArray_4 = RigVMDispatchNode(Value=((Type=Bone,Name="dyn_hair_bk_r_01"),(Type=Bone,Name="dyn_hair_bk_r_02"),(Type=Bone,Name="dyn_hair_bk_r_03"),(Type=Bone,Name="dyn_hair_bk_r_04"),(Type=Bone,Name="dyn_hair_bk_r_05"),(Type=Bone,Name="dyn_hair_bk_r_06"))) {template="DISPATCH_RigVMDispatch_Constant(io Value)"}
   ItemArray_5 = RigVMDispatchNode(Value="((Type=Control,Name=\"dyn_hair_bk_r_01_ctrl\"),(Type=Control,Name=\"dyn_hair_bk_r_02_ctrl\"),(Type=Control,Name=\"dyn_hair_bk_r_03_ctrl\"),(Type=Control,Name=\"dyn_hair_bk_r_04_ctrl\"),(Type=Control,Name=\"dyn_hair_bk_r_05_ctrl\"),(Type=Control,Name=\"dyn_hair_bk_r_0 ...(truncated)") {template="DISPATCH_RigVMDispatch_Constant(io Value)"}
   > SplineIK_2 = RigVMFunctionReferenceNode(
    Primary Axis=(X=1,Y=0,Z=0)
    Samples Per Segment=16
    Secondary Spline Direction=(X=0,Y=5,Z=0)
    Spline Mode=BSpline
    Up Axis=(X=0,Y=1,Z=0)
    Bones=@ItemArray_4.Value
    Controls=@ItemArray_5.Value
   )
   n2 = &RigUnit_SphericalPoseReader(
    ActiveRegionScaleFactors=(PositiveWidth=0.5,NegativeWidth=0.5,PositiveHeight=0.5,NegativeHeight=0.5)
    ActiveRegionSize=0.12
    Debug=(bDrawDebug=false,bDraw2D=False,bDrawLocalAxes=False,DebugScale=25,DebugSegments=20,DebugThickness=0.25)
    DriverAxis=(X=1,Y=0,Z=0)
    DriverItem=(Type=Bone,Name="clavicle_l")
    FalloffRegionScaleFactors=(PositiveWidth=0.5,NegativeWidth=0.5,PositiveHeight=0.5,NegativeHeight=0.5)
    FalloffSize=0.3
    FlipHeightScaling=False
    FlipWidthScaling=False
    OptionalParentItem=(Type=Bone,Name="None")
    RotationOffset=(X=0,Y=-90,Z=-40)
    ExecutePin=@RigVMFunction_Sequence.B
   )
   > &RigUnit_OffsetTransformForItem(
    Item=(Type=Bone,Name="dyn_hair_bk_l_01")
    OffsetTransform=(Rotation=(X=0,Y=0,Z=0,W=1),Translation=(X=0,Y=0,Z=2),Scale3D=(X=1,Y=1,Z=1))
    bPropagateToChildren=false
    Weight=@n2.OutputParam
   )
   > &RigUnit_OffsetTransformForItem(
    Item=(Type=Bone,Name="dyn_hair_bk_l_02")
    OffsetTransform=(Rotation=(X=0,Y=0,Z=0,W=1),Translation=(X=0,Y=0,Z=2),Scale3D=(X=1,Y=1,Z=1))
    bPropagateToChildren=false
    Weight=@n2.OutputParam
   )
   > &RigUnit_OffsetTransformForItem(
    Item=(Type=Bone,Name="dyn_hair_bk_l_03")
    OffsetTransform=(Rotation=(X=0,Y=0,Z=0,W=-1),Translation=(X=0,Y=0,Z=2),Scale3D=(X=1,Y=1,Z=1))
    bPropagateToChildren=false
    Weight=@n2.OutputParam
   )
   > &RigUnit_OffsetTransformForItem(
    Item=(Type=Bone,Name="dyn_hair_bk_02")
    OffsetTransform=(Rotation=(X=0,Y=0,Z=0,W=-1),Translation=(X=0,Y=1,Z=3),Scale3D=(X=1,Y=1,Z=1))
    bPropagateToChildren=true
    Weight=@n2.OutputParam
   )
   > n3 = &RigUnit_SphericalPoseReader(
    ActiveRegionScaleFactors=(PositiveWidth=0.5,NegativeWidth=0.5,PositiveHeight=0.5,NegativeHeight=0.5)
    ActiveRegionSize=0.12
    Debug=(bDrawDebug=false,bDraw2D=False,bDrawLocalAxes=False,DebugScale=25,DebugSegments=20,DebugThickness=0.25)
    DriverAxis=(X=1,Y=0,Z=0)
    DriverItem=(Type=Bone,Name="clavicle_r")
    FalloffRegionScaleFactors=(PositiveWidth=0.5,NegativeWidth=0.5,PositiveHeight=0.5,NegativeHeight=0.5)
    FalloffSize=0.3
    FlipHeightScaling=False
    FlipWidthScaling=False
    OptionalParentItem=(Type=Bone,Name="None")
    RotationOffset=(X=0,Y=-90,Z=-32)
   )
   > &RigUnit_OffsetTransformForItem(
    Item=(Type=Bone,Name="dyn_hair_bk_r_02")
    OffsetTransform=(Rotation=(X=0,Y=0,Z=0,W=1),Translation=(X=0,Y=-1,Z=5),Scale3D=(X=1,Y=1,Z=1))
    bPropagateToChildren=false
    Weight=@n3.OutputParam
   )
   > &RigUnit_OffsetTransformForItem(
    Item=(Type=Bone,Name="dyn_hair_bk_r_03")
    OffsetTransform=(Rotation=(X=0,Y=0,Z=0,W=-1),Translation=(X=0,Y=1,Z=4),Scale3D=(X=1,Y=1,Z=1))
    bPropagateToChildren=false
    Weight=@n3.OutputParam
   )
   > &RigUnit_OffsetTransformForItem(
    Item=(Type=Bone,Name="dyn_hair_bk_02")
    OffsetTransform=(Rotation=(X=0,Y=0,Z=0,W=-1),Translation=(X=0,Y=1,Z=1.5),Scale3D=(X=1,Y=1,Z=1))
    bPropagateToChildren=true
    Weight=@n3.OutputParam
   )
   > Delta Rotation Y = RigVMFunctionReferenceNode(Child=clavicle_l, Parent=clavicle_l_null, Source Maximum=-70)
   > &RigUnit_OffsetTransformForItem(
    Item=(Type=Bone,Name="dyn_hair_bk_l_02")
    OffsetTransform=(Rotation=(X=0,Y=0,Z=0,W=1),Translation=(X=0,Y=0,Z=4),Scale3D=(X=1,Y=1,Z=1))
    bPropagateToChildren=false
    Weight=@"Delta Rotation Y".Delta Rotation
   )
   > Delta Rotation Y_1 = RigVMFunctionReferenceNode(Child=clavicle_r, Parent=clavicle_r_null, Source Maximum=-70)
   > &RigUnit_OffsetTransformForItem(
    Item=(Type=Bone,Name="dyn_hair_bk_r_02")
    OffsetTransform=(Rotation=(X=0,Y=0,Z=0,W=1),Translation=(X=-3,Y=0,Z=4),Scale3D=(X=1,Y=1,Z=1))
    bPropagateToChildren=false
    Weight=@"Delta Rotation Y_1".Delta Rotation
   )
   > &RigUnit_OffsetTransformForItem(
    Item=(Type=Bone,Name="dyn_hair_bk_r_01")
    OffsetTransform=(Rotation=(X=0,Y=0,Z=0,W=1),Translation=(X=-3,Y=0,Z=3),Scale3D=(X=1,Y=1,Z=1))
    bPropagateToChildren=false
    Weight=@"Delta Rotation Y_1".Delta Rotation
   )
   > &RigUnit_OffsetTransformForItem(
    Item=(Type=Bone,Name="dyn_hair_r_01")
    OffsetTransform=(Rotation=(X=0,Y=0,Z=-0.0785,W=0.9969),Translation=(X=-1.6727,Y=-4,Z=0),Scale3D=(X=1,Y=1,Z=1))
    bPropagateToChildren=true
    Weight=@"Delta Rotation Y_1".Delta Rotation
   )
   > Delta Rotation Y_2 = RigVMFunctionReferenceNode(Child=clavicle_l, Parent=clavicle_l_null, Source Maximum=-42)
   RigVMFunction_MathBoolToFloat = &RigVMFunction_MathBoolToFloat(Value=RigVMVariableNode() {variable=Skydive}.Value)
   AlphaInterp = &RigVMFunction_AlphaInterp(InterpSpeedDecreasing=10, InterpSpeedIncreasing=10, Scale=1, bClampResult=true, bMapRange=true, Value=@RigVMFunction_MathBoolToFloat.Result)
   Multiply = &RigVMFunction_MathDoubleMul(B=@AlphaInterp.Result, A=@"Delta Rotation Y_2".Delta Rotation)
   > n4 = &RigUnit_ModifyTransforms(
    ItemToModify=((Item=(Type=Bone,Name="dyn_hair_bk_l_01"),Transform=(Rotation=(X=0,Y=-0.0168,Z=0,W=-0.9999),Translation=(X=-1.2748,Y=-0.0001,Z=4.7102),Scale3D=(X=1,Y=1,Z=1))))
    WeightMinimum=0
    Weight=@Multiply.Result
   )
   RigVMFunction_ControlFlowBranch = &RigVMFunction_ControlFlowBranch(ExecuteContext=@n4.ExecutePin, Condition=RigVMVariableNode() {variable=wearingBP}.Value)
   n5 = &RigUnit_SphericalPoseReader(
    ActiveRegionScaleFactors=(PositiveWidth=1,NegativeWidth=1,PositiveHeight=1,NegativeHeight=0.8)
    ActiveRegionSize=0.1
    Debug=(bDrawDebug=false,bDraw2D=False,bDrawLocalAxes=False,DebugScale=25,DebugSegments=20,DebugThickness=0.25)
    DriverAxis=(X=1,Y=0,Z=0)
    DriverItem=(Type=Bone,Name="upperarm_r")
    FalloffRegionScaleFactors=(PositiveWidth=1,NegativeWidth=1,PositiveHeight=1.7,NegativeHeight=0.5)
    FalloffSize=0.2
    FlipHeightScaling=False
    FlipWidthScaling=False
    OptionalParentItem=(Type=Bone,Name="None")
    RotationOffset=(X=51.9395,Y=8.7023,Z=22.4256)
    ExecutePin=@RigVMFunction_ControlFlowBranch.Completed
   )
   > &RigUnit_ModifyTransforms(
    ItemToModify="((Item=(Type=Bone,Name=\"base_hair_bk_01\"),Transform=(Rotation=(X=0,Y=0.2588,Z=0,W=0.9659),Translation=(X=-2.063,Y=0,Z=-0.5733),Scale3D=(X=1,Y=1,Z=1))),(Item=(Type=Bone,Name=\"base_hair_bk_l_01\"),Transform=(Rotation=(X=0,Y=0.2588,Z=0,W=0.9659),Translation=(X ...(truncated)"
    WeightMinimum=0
    Weight=@n5.OutputParam
   )
   Delta Rotation X = RigVMFunctionReferenceNode(Child=head, Parent=head_null_2, Source Maximum=-33, ExecuteContext=@RigVMFunction_ControlFlowBranch.False)
   > &RigUnit_ModifyTransforms(
    ItemToModify="((Item=(Type=Bone,Name=\"base_hair_l_01\"),Transform=(Rotation=(X=0,Y=0,Z=0,W=-1),Translation=(X=0,Y=-2.4778,Z=0),Scale3D=(X=1,Y=1,Z=1))),(Item=(Type=Bone,Name=\"base_hair_bk_l_01\"),Transform=(Rotation=(X=0,Y=0,Z=0,W=1),Translation=(X=0,Y=2.9886,Z=1.1579),Sca ...(truncated)"
    WeightMinimum=0
    Weight=@"Delta Rotation X".Delta Rotation
   )
   > HeadAdjust = RigVMFunctionReferenceNode()
   > Delta Rotation Y_4 = RigVMFunctionReferenceNode(Child=clavicle_l, Parent=clavicle_l_null, Source Maximum=-20, Source Minimum=-5)
   n6 = &RigVMFunction_AlphaInterp(
    InterpSpeedDecreasing=10
    InterpSpeedIncreasing=10
    Scale=1
    bClampResult=true
    bMapRange=true
    Value=&RigVMFunction_MathBoolToFloat(Value=RigVMVariableNode() {variable=Frontend}.Value).Result
   )
   n7 = &RigVMFunction_AlphaInterp(
    InterpSpeedDecreasing=10
    InterpSpeedIncreasing=10
    Scale=1
    bClampResult=true
    bMapRange=true
    Value=&RigVMFunction_MathBoolToFloat(Value=RigVMVariableNode() {variable=PlayingEmote}.Value).Result
   )
   Multiply_1 = RigVMAggregateNode(
    B=@n6.Result
    A=@"Delta Rotation Y_4".Delta Rotation
    C=&RigVMFunction_MathDoubleRemap(SourceMaximum=1, SourceMinimum=0, TargetMaximum=0, TargetMinimum=1, Value=@n7.Result).Result
   )
   > &RigUnit_ModifyTransforms(
    ItemToModify="((Item=(Type=Bone,Name=\"dyn_hair_bk_l_01\"),Transform=(Rotation=(X=0,Y=0,Z=0,W=1),Translation=(X=-2.8183,Y=0,Z=4.436),Scale3D=(X=1,Y=1,Z=1))),(Item=(Type=Bone,Name=\"dyn_hair_l_01\"),Transform=(Rotation=(X=0,Y=0,Z=0,W=1),Translation=(X=0,Y=-1.645,Z=0),Scale3D ...(truncated)"
    WeightMinimum=0
    Weight=@Multiply_1.Result
   )
   &RigUnit_ParentConstraint(
    AdvancedSettings=(InterpolationType=Average,RotationOrderForFilter=XZY)
    Child=(Type=Bone,Name="dyn_hair_bk_01")
    Parents=((Item=(Type=Bone,Name="dyn_hair_bk_l_01"),Weight=0.5),(Item=(Type=Bone,Name="dyn_hair_bk_r_01"),Weight=0.5))
    Weight=1
    ExecutePin=@RigVMFunction_Sequence.C
   )
   > &RigUnit_ParentConstraint(
    AdvancedSettings=(InterpolationType=Average,RotationOrderForFilter=XZY)
    Child=(Type=Bone,Name="dyn_hair_bk_02")
    Parents=((Item=(Type=Bone,Name="dyn_hair_bk_l_02"),Weight=0.5),(Item=(Type=Bone,Name="dyn_hair_bk_r_02"),Weight=0.5))
    Weight=1
   )
   > &RigUnit_ParentConstraint(
    AdvancedSettings=(InterpolationType=Average,RotationOrderForFilter=XZY)
    Child=(Type=Bone,Name="dyn_hair_bk_l_03")
    Parents=((Item=(Type=Bone,Name="dyn_hair_bk_l_03"),Weight=0.5),(Item=(Type=Bone,Name="dyn_hair_bk_r_03"),Weight=0.5))
    Weight=1
   )
   > &RigUnit_ParentConstraint(
    AdvancedSettings=(InterpolationType=Average,RotationOrderForFilter=XZY)
    Child=(Type=Bone,Name="dyn_hair_bk_04")
    Parents=((Item=(Type=Bone,Name="dyn_hair_bk_r_04"),Weight=0.5),(Item=(Type=Bone,Name="dyn_hair_bk_r_04"),Weight=0.5))
    Weight=1
   )
   > &RigUnit_ParentConstraint(
    AdvancedSettings=(InterpolationType=Average,RotationOrderForFilter=XZY)
    Child=(Type=Bone,Name="dyn_hair_bk_05")
    Parents=((Item=(Type=Bone,Name="dyn_hair_bk_l_05"),Weight=0.5),(Item=(Type=Bone,Name="dyn_hair_bk_r_05"),Weight=0.5))
    Weight=1
   )
   > &RigUnit_ParentConstraint(
    AdvancedSettings=(InterpolationType=Average,RotationOrderForFilter=XZY)
    Child=(Type=Bone,Name="dyn_hair_bk_06")
    Parents=((Item=(Type=Bone,Name="dyn_hair_bk_l_06"),Weight=0.5),(Item=(Type=Bone,Name="dyn_hair_bk_r_06"),Weight=0.5))
    Weight=1
   )
   > &RigUnit_ParentConstraint(
    AdvancedSettings=(InterpolationType=Average,RotationOrderForFilter=XZY)
    Child=(Type=Bone,Name="dyn_hair_bk_07")
    Parents=((Item=(Type=Bone,Name="dyn_hair_bk_l_06"),Weight=0.5),(Item=(Type=Bone,Name="dyn_hair_bk_r_06"),Weight=0.5))
    Weight=1
   )
   n8 = &RigVMFunction_ControlFlowBranch(Condition=RigVMVariableNode() {variable=PlayingEmote}.Value, ExecuteContext=@RigVMFunction_Sequence.D)
   Delta Rotation Z = RigVMFunctionReferenceNode(Child=head, Parent=head_null, Source Maximum=-40, ExecuteContext=@n8.False)
   > &RigUnit_OffsetTransformForItem(
    Item=(Type=Bone,Name="dyn_hair_bk_l_01")
    OffsetTransform=(Rotation=(X=0,Y=-0.1181,Z=0,W=0.993),Translation=(X=0.21,Y=0.0001,Z=4.5649),Scale3D=(X=1,Y=1,Z=1))
    bPropagateToChildren=false
    Weight=@"Delta Rotation Z".Delta Rotation
   )
   > &RigUnit_OffsetTransformForItem(
    Item=(Type=Bone,Name="dyn_hair_bk_01")
    OffsetTransform=(Rotation=(X=0,Y=0,Z=0,W=1),Translation=(X=6.2578,Y=0,Z=5.1171),Scale3D=(X=1,Y=1,Z=1))
    bPropagateToChildren=false
    Weight=@"Delta Rotation Z".Delta Rotation
   )
   > &RigUnit_OffsetTransformForItem(
    Item=(Type=Bone,Name="dyn_hair_bk_r_01")
    OffsetTransform=(Rotation=(X=0,Y=0.059,Z=0,W=0.9983),Translation=(X=1,Y=0,Z=7.5553),Scale3D=(X=1,Y=1,Z=1))
    bPropagateToChildren=true
    Weight=@"Delta Rotation Z".Delta Rotation
   )
   > &RigUnit_OffsetTransformForItem(
    Item=(Type=Bone,Name="dyn_hair_bk_l_02")
    OffsetTransform=(Rotation=(X=0,Y=0,Z=0,W=1),Translation=(X=0,Y=0,Z=2),Scale3D=(X=1,Y=1,Z=1))
    bPropagateToChildren=true
    Weight=@"Delta Rotation Z".Delta Rotation
   )
   > &RigUnit_OffsetTransformForItem(
    Item=(Type=Bone,Name="dyn_hair_bk_02")
    OffsetTransform=(Rotation=(X=0,Y=0,Z=0,W=1),Translation=(X=0,Y=0,Z=0),Scale3D=(X=1,Y=1,Z=1))
    bPropagateToChildren=true
    Weight=@"Delta Rotation Z".Delta Rotation
   )
   > n9 = &RigUnit_OffsetTransformForItem(
    Item=(Type=Bone,Name="dyn_hair_bk_r_01")
    OffsetTransform=(Rotation=(X=0,Y=0,Z=0,W=1),Translation=(X=0,Y=0,Z=0),Scale3D=(X=1,Y=1,Z=1))
    bPropagateToChildren=true
    Weight=@"Delta Rotation Z".Delta Rotation
   )
   n10 = &RigVMFunction_ControlFlowBranch(ExecuteContext=@n9.ExecutePin, Condition=RigVMVariableNode() {variable=IsCrouching}.Value)
   n11 = &RigUnit_SphericalPoseReader(
    ActiveRegionScaleFactors=(PositiveWidth=0.5,NegativeWidth=0.5,PositiveHeight=0.3,NegativeHeight=0.15)
    ActiveRegionSize=0.12
    Debug=(bDrawDebug=false,bDraw2D=False,bDrawLocalAxes=False,DebugScale=25,DebugSegments=20,DebugThickness=0.25)
    DriverAxis=(X=0,Y=1,Z=0)
    DriverItem=(Type=Bone,Name="spine_05")
    FalloffRegionScaleFactors=(PositiveWidth=0.5,NegativeWidth=0.5,PositiveHeight=0.15,NegativeHeight=0.2)
    FalloffSize=0.3
    FlipHeightScaling=False
    FlipWidthScaling=False
    OptionalParentItem=(Type=Bone,Name="spine_03")
    RotationOffset=(X=71,Y=0,Z=0)
    ExecutePin=@n10.True
   )
   > &RigUnit_ModifyTransforms(
    ItemToModify=((Item=(Type=Bone,Name="base_hair_bk_r_01"),Transform=(Rotation=(X=0,Y=0,Z=0,W=1),Translation=(X=-2,Y=-1.1475,Z=3.7948),Scale3D=(X=1,Y=1,Z=1))))
    WeightMinimum=0
    Weight=@n11.OutputParam
   )
   Remap = &RigVMFunction_MathDoubleRemap(SourceMaximum=0.6, SourceMinimum=0.2, TargetMaximum=1, TargetMinimum=0, Value=@"Delta Rotation Z".Delta Rotation)
   > &RigUnit_ModifyTransforms(
    ItemToModify=((Item=(Type=Bone,Name="base_hair_bk_r_01"),Transform=(Rotation=(X=-0.0976,Y=-0.0278,Z=-0.1705,W=0.9801),Translation=(X=-1.8674,Y=-2,Z=-5.5239),Scale3D=(X=1,Y=1,Z=1))))
    WeightMinimum=0
    Weight=&RigVMFunction_MathDoubleMul(A=@Remap.Result, B=@n11.OutputParam).Result
   )
   Delta Rotation Z_1 = RigVMFunctionReferenceNode(Child=head, Parent=head_null, Source Maximum=-40, ExecuteContext=@n8.True)
   > &RigUnit_OffsetTransformForItem(
    Item=(Type=Bone,Name="dyn_hair_bk_l_01")
    OffsetTransform=(Rotation=(X=0.0009,Y=0.1181,Z=-0.0079,W=-0.993),Translation=(X=-5.0135,Y=1.6711,Z=0.1368),Scale3D=(X=1,Y=1,Z=1))
    bPropagateToChildren=false
    Weight=@"Delta Rotation Z_1".Delta Rotation
   )
   > &RigUnit_OffsetTransformForItem(
    Item=(Type=Bone,Name="dyn_hair_bk_r_01")
    OffsetTransform=(Rotation=(X=0,Y=-0.0393,Z=0,W=-0.9992),Translation=(X=-4.1628,Y=-2.5582,Z=4.5605),Scale3D=(X=1,Y=1,Z=1))
    bPropagateToChildren=true
    Weight=@"Delta Rotation Z_1".Delta Rotation
   )
   > Delta Rotation Y_3 = RigVMFunctionReferenceNode(Child=clavicle_l, Parent=clavicle_l_null, Source Maximum=-45)
   > &RigUnit_OffsetTransformForItem(
    Item=(Type=Bone,Name="dyn_hair_bk_l_01")
    OffsetTransform=(Rotation=(X=0,Y=0,Z=-0.2164,W=0.9763),Translation=(X=-3,Y=2,Z=2),Scale3D=(X=1,Y=1,Z=1))
    bPropagateToChildren=false
    Weight=@"Delta Rotation Y_3".Delta Rotation
   )
   n12 = &RigVMFunction_AlphaInterp(
    InterpSpeedDecreasing=10
    InterpSpeedIncreasing=10
    Scale=1
    bClampResult=true
    bMapRange=true
    Value=&RigVMFunction_MathBoolToFloat(Value=RigVMVariableNode() {variable=Frontend}.Value).Result
   )
   > &RigUnit_ModifyTransforms(
    ItemToModify=((Item=(Type=Bone,Name="None"),Transform=(Rotation=(X=0,Y=0,Z=0,W=1),Translation=(X=0,Y=0,Z=0),Scale3D=(X=1,Y=1,Z=1))),(Item=(Type=Bone,Name="base_hair_l_01"),Transform=(Rotation=(X=0,Y=0,Z=0,W=1),Translation=(X=0,Y=-6,Z=1),Scale3D=(X=1,Y=1,Z=1))))
    WeightMinimum=0
    Weight=&RigVMFunction_MathDoubleMul(A=@n12.Result, B=@"Delta Rotation Y_3".Delta Rotation).Result
   )
   FN_DeformRig_FNC = RigVMFunctionReferenceNode(
    DeformRig_ Curve="(deformRig_l_upperarm_up_alpha_72_#8=False,deformRig_r_upperarm_up_alpha_73_#9=False,deformRig_l_upperarm_fr_alpha_74_#10=False,deformRig_r_upperarm_fr_alpha_75_CB51B8 ...(truncated)"
    DeformRig_Bool="(upperarm_1_#11=True,elbow_3_#12=True,wrist_7_#13=True,groin_8_#14=True,glute_17_#15=True,knee_9_40C408324CA261 ...(truncated)"
    DeformRig_Debug=(upperarm_debug_18_#11=false,wrist_debug_20_#13=False,lat_debug_26_#16=False,groinGlute_debug_27_#14=False)
    DeformRig_Value="(upperarm_up_18_#17=(Rotation=(X=0,Y=0,Z=0,W=1),Translation=(X=0,Y=0,Z=8),Scale3D=(X=0.7,Y=1,Z=1)),upperarm_fwd_fr_19_#18=(Rotation=(X=0,Y=0,Z=0,W=1),Translation=(X=0,Y=10,Z=0),Scale3D=(X=0.5,Y=1,Z= ...(truncated)"
    ExecuteContext=()
   )
  ]}
]
hierarchy [
 root {loc=<0 0 0>, rot=<0 0 0 1>}
 pelvis {parent=root, loc=<0 -0.5301 98.0571>, rot=<0.0127 0.707 -0.0127 -0.707>}
 spine_01 {parent=pelvis, loc=<6.1208 -0.9328 0>, rot=<0 0 0.034 -0.9994>}
 spine_02 {parent=spine_01, loc=<6.5997 -0.2095 0>, rot=<0 0 0.0238 -0.9997>}
 spine_03 {parent=spine_02, loc=<6.4914 -0.1237 0>, rot=<0 0 0.0149 -0.9999>}
 spine_04 {parent=spine_03, loc=<6.9598 0.0794 0>, rot=<0 0 -0.0132 -0.9999>}
 spine_05 {parent=spine_04, loc=<6.9987 0 0>, rot=<0 0 -0.0015 -1>}
 neck_01 {parent=spine_05, loc=<11.3138 0.1529 0>, rot=<0 0 -0.2003 -0.9797>}
 neck_02 {parent=neck_01, loc=<4.5627 0 0>, rot=<0 0 0.0504 -0.9987>}
 head {parent=neck_02, loc=<5.3914 0 0>, rot=<0 0 0.115 -0.9934>}
 base_hair_l_01 {parent=head, loc=<0.8111 -0.9382 -7.2809>, rot=<-0.0207 0.13 0.991 -0.0237>}
 dyn_hair_l_01 {parent=base_hair_l_01, loc=<0 0 0>, rot=<0 0 0 1>}
 dyn_hair_l_02 {parent=dyn_hair_l_01, loc=<4.8358 0 0>, rot=<0 0 0.1564 -0.9877>}
 dyn_hair_l_03 {parent=dyn_hair_l_02, loc=<4.8487 0 0>, rot=<0 0 0.1045 -0.9945>}
 dyn_hair_l_04 {parent=dyn_hair_l_03, loc=<4.853 0 0>, rot=<0 0 -0.0087 -1>}
 dyn_hair_l_05 {parent=dyn_hair_l_04, loc=<4.8496 0 0>, rot=<0 0 -0.1045 -0.9945>}
 dyn_hair_l_06 {parent=dyn_hair_l_05, loc=<4.5 0 0>, rot=<0 0 -0.0262 -0.9997>}
 dyn_hair_l_07 {parent=dyn_hair_l_06, loc=<4.5 0 0>, rot=<0 0 0 1>}
 base_hair_r_01 {parent=head, loc=<0.8271 -0.945 6.8925>, rot=<0.1002 0.1325 0.9853 -0.0394>}
 dyn_hair_r_01 {parent=base_hair_r_01, loc=<0 0 0>, rot=<0 0 0 1>}
 dyn_hair_r_02 {parent=dyn_hair_r_01, loc=<4.85 0 0>, rot=<0 0 0.1564 -0.9877>}
 dyn_hair_r_03 {parent=dyn_hair_r_02, loc=<4.85 0 0>, rot=<0 0 0.1045 -0.9945>}
 dyn_hair_r_04 {parent=dyn_hair_r_03, loc=<4.85 0 0>, rot=<0 0 -0.0087 -1>}
 dyn_hair_r_05 {parent=dyn_hair_r_04, loc=<4.85 0 0>, rot=<0 0 -0.1045 -0.9945>}
 dyn_hair_r_06 {parent=dyn_hair_r_05, loc=<4.5 0 0>, rot=<0 0 -0.0262 -0.9997>}
 dyn_hair_r_07 {parent=dyn_hair_r_06, loc=<4.5 0 0>, rot=<0 0 0 1>}
 earring_r {parent=head, loc=<0.6267 0.3622 6.2074>, rot=<0 0 -0.0173 -0.9999>}
 earring_l {parent=head, loc=<0.6267 0.3622 -6.2074>, rot=<0 0 -0.0173 -0.9999>}
 base_hair_bk_01 {parent=head, loc=<1.2116 -5.0574 0>, rot=<0.1559 -0.6897 0.6897 -0.1559>}
 dyn_hair_bk_01 {parent=base_hair_bk_01, loc=<0 0 0>, rot=<0 0 0 -1>}
 dyn_hair_bk_02 {parent=dyn_hair_bk_01, loc=<7.3205 0 0>, rot=<0 -0.0523 0 -0.9986>}
 dyn_hair_bk_03 {parent=dyn_hair_bk_02, loc=<7.3197 0 0>, rot=<0 -0.1219 0 -0.9925>}
 dyn_hair_bk_04 {parent=dyn_hair_bk_03, loc=<7.2884 0 0>, rot=<0 0 0 1>}
 dyn_hair_bk_05 {parent=dyn_hair_bk_04, loc=<7.2939 0 0>, rot=<0 -0.0785 0 -0.9969>}
 dyn_hair_bk_06 {parent=dyn_hair_bk_05, loc=<7.3213 0 0>, rot=<0 0.0087 0 -1>}
 dyn_hair_bk_07 {parent=dyn_hair_bk_06, loc=<7.3211 0 0>, rot=<0 0 0 1>}
 base_hair_bk_l_01 {parent=head, loc=<1.2116 -5.0574 -5>, rot=<0.0952 -0.7007 0.6735 -0.2154>}
 dyn_hair_bk_l_01 {parent=base_hair_bk_l_01, loc=<0 0 0>, rot=<0 0 0 1>}
 dyn_hair_bk_l_02 {parent=dyn_hair_bk_l_01, loc=<7.3205 0 0>, rot=<0 -0.0523 0 -0.9986>}
 dyn_hair_bk_l_03 {parent=dyn_hair_bk_l_02, loc=<7.3197 0 0>, rot=<-0.0043 -0.1218 0.0346 -0.9919>}
 dyn_hair_bk_l_04 {parent=dyn_hair_bk_l_03, loc=<7.2884 0 0>, rot=<-0.0018 -0.0523 0.0349 -0.998>}
 dyn_hair_bk_l_05 {parent=dyn_hair_bk_l_04, loc=<7.2939 0 0>, rot=<0 0 0 1>}
 dyn_hair_bk_l_06 {parent=dyn_hair_bk_l_05, loc=<7.3213 0 0>, rot=<0 0.1045 0 -0.9945>}
 base_hair_bk_r_01 {parent=head, loc=<1.2116 -5.0574 6.25>, rot=<0.2154 -0.6735 0.7007 -0.0952>}
 dyn_hair_bk_r_01 {parent=base_hair_bk_r_01, loc=<0 0 0>, rot=<0 0 0 1>}
 dyn_hair_bk_r_02 {parent=dyn_hair_bk_r_01, loc=<7.3205 0 0>, rot=<0 -0.0523 0 -0.9986>}
 dyn_hair_bk_r_03 {parent=dyn_hair_bk_r_02, loc=<7.3197 0 0>, rot=<0.0043 -0.1218 -0.0346 -0.9919>}
 dyn_hair_bk_r_04 {parent=dyn_hair_bk_r_03, loc=<7.2884 0 0>, rot=<0.0011 -0.0436 -0.0262 -0.9987>}
 dyn_hair_bk_r_05 {parent=dyn_hair_bk_r_04, loc=<7.2939 0 0>, rot=<0 0 0 1>}
 dyn_hair_bk_r_06 {parent=dyn_hair_bk_r_05, loc=<7.3213 0 0>, rot=<0 0.1045 0 -0.9945>}
 head_null_2 {parent=neck_02, type=Null, loc=<5.3914 0 0>, rot=<0 0 0.115 -0.9934>}
 head_null {parent=neck_01, type=Null, loc=<9.9266 -0.5431 0>, rot=<0 0 -0.1649 0.9863>}
 clavicle_l {parent=spine_05, loc=<5.4831 -0.1053 -3.2946>, rot=<0.0465 -0.7289 -0.009 -0.6829>}
 upperarm_l {parent=clavicle_l, loc=<9.8658 -1.2136 -0.4105>, rot=<0.0164 -0.3245 -0.0269 -0.9454>}
 clavicle_r {parent=spine_05, loc=<5.4831 -0.1053 3.2946>, rot=<0.7289 0.0465 -0.6829 0.009>}
 upperarm_r {parent=clavicle_r, loc=<-9.8658 1.2136 0.4105>, rot=<0.0164 -0.3245 -0.0269 -0.9454>}
 upperarm_r_null {parent=clavicle_r, type=Null, loc=<-9.8658 1.2136 0.4105>, rot=<0.0164 -0.3245 -0.0269 -0.9454>}
 clavicle_l_null {parent=spine_05, type=Null, loc=<5.4831 -0.1053 -3.2946>, rot=<0.0465 -0.7289 -0.009 -0.6829>}
 clavicle_r_null {parent=spine_05, type=Null, loc=<5.4831 -0.1053 3.2946>, rot=<0.7289 0.0465 -0.6829 0.009>}
 dyn_hair_bk_01_ctrl_Null {type=Null, loc=<0 -6.7546 153.204>, rot=<-0.3799 -0.3799 0.5964 -0.5964>}
 dyn_hair_bk_01_ctrl {controlType=EULER_TRANSFORM, parent=dyn_hair_bk_01_ctrl_Null, type=Control, loc=<0 0 0>, rot=<0 0 0 1>}
 dyn_hair_bk_02_ctrl_Null {type=Null, loc=<0 -9.8483 146.5694>, rot=<-0.4106 -0.4106 0.5757 -0.5757>}
 dyn_hair_bk_02_ctrl {
  controlType=EULER_TRANSFORM
  parent=dyn_hair_bk_02_ctrl_Null
  type=Control
  loc=<-22.601 -9.8483 145.0013>
  rot=<0.4402 0.3799 -0.5534 0.5964>
 }
 dyn_hair_bk_03_ctrl_Null {type=Null, loc=<0 -12.2314 139.6485>, rot=<-0.4777 -0.4777 0.5213 -0.5213>}
 dyn_hair_bk_03_ctrl {
  controlType=EULER_TRANSFORM
  parent=dyn_hair_bk_03_ctrl_Null
  type=Control
  loc=<-40.8863 -12.2314 133.7296>
  rot=<0.5377 0.4106 -0.4592 0.5757>
 }
 dyn_hair_bk_04_ctrl_Null {type=Null, loc=<0 -12.8666 132.3878>, rot=<-0.5456 -0.5456 0.4498 -0.4498>}
 dyn_hair_bk_04_ctrl {
  controlType=EULER_TRANSFORM
  parent=dyn_hair_bk_04_ctrl_Null
  type=Control
  loc=<-43.4971 -12.8666 125.2504>
  rot=<0.6029 0.4777 -0.3695 0.5213>
 }
 dyn_hair_bk_05_ctrl_Null {type=Null, loc=<0 -11.4749 125.2279>, rot=<-0.5456 -0.5456 0.4498 -0.4498>}
 dyn_hair_bk_05_ctrl {
  controlType=EULER_TRANSFORM
  parent=dyn_hair_bk_05_ctrl_Null
  type=Control
  loc=<-7.2939 -11.4749 125.2279>
  rot=<-0.5456 -0.5456 0.4498 -0.4498>
 }
 dyn_hair_bk_06_ctrl_Null {type=Null, loc=<0 -10.0779 118.0411>, rot=<-0.4956 -0.4956 0.5043 -0.5043>}
 dyn_hair_bk_06_ctrl {
  controlType=EULER_TRANSFORM
  parent=dyn_hair_bk_06_ctrl_Null
  type=Control
  loc=<17.3808 -10.0779 116.9838>
  rot=<0.4402 0.5456 -0.5534 0.4498>
 }
 dyn_hair_bk_07_ctrl_Null {type=Null, loc=<0 -10.2057 110.7211>, rot=<0.4956 0.4956 -0.5043 0.5043>}
 dyn_hair_bk_07_ctrl {
  controlType=EULER_TRANSFORM
  parent=dyn_hair_bk_07_ctrl_Null
  type=Control
  loc=<-7.3211 -10.2057 110.7211>
  rot=<0.4956 0.4956 -0.5043 0.5043>
 }
 dyn_hair_bk_l_01_ctrl_Null {type=Null, loc=<5 -6.7546 153.204>, rot=<-0.4116 -0.3454 0.5421 -0.6461>}
 dyn_hair_bk_l_01_ctrl {controlType=EULER_TRANSFORM, parent=dyn_hair_bk_l_01_ctrl_Null, type=Control, loc=<5 -6.7546 153.204>, rot=<0.4116 0.3454 -0.5421 0.6461>}
 dyn_hair_bk_l_02_ctrl_Null {type=Null, loc=<6.2712 -9.8013 146.6702>, rot=<-0.4394 -0.3787 0.5198 -0.6271>}
 dyn_hair_bk_l_02_ctrl {
  controlType=EULER_TRANSFORM
  parent=dyn_hair_bk_l_02_ctrl_Null
  type=Control
  loc=<-16.3747 -9.8013 145.7571>
  rot=<0.466 0.3454 -0.4961 0.6461>
 }
 dyn_hair_bk_l_03_ctrl_Null {type=Null, loc=<7.5353 -12.1376 139.8496>, rot=<-0.4887 -0.465 0.4855 -0.5561>}
 dyn_hair_bk_l_03_ctrl {
  controlType=EULER_TRANSFORM
  parent=dyn_hair_bk_l_03_ctrl_Null
  type=Control
  loc=<-32.8025 -12.093 135.9523>
  rot=<0.5254 0.4084 -0.4047 0.6271>
 }
 dyn_hair_bk_l_04_ctrl_Null {type=Null, loc=<8.2358 -12.7596 132.6216>, rot=<-0.5271 -0.5623 0.4495 -0.4515>}
 dyn_hair_bk_l_04_ctrl {
  controlType=EULER_TRANSFORM
  parent=dyn_hair_bk_l_04_ctrl_Null
  type=Control
  loc=<-34.1568 -12.549 128.1719>
  rot=<0.5456 0.5211 -0.3486 0.5561>
 }
 dyn_hair_bk_l_05_ctrl_Null {type=Null, loc=<7.9691 -11.3965 125.4611>, rot=<0.5271 0.5623 -0.4495 0.4515>}
 dyn_hair_bk_l_05_ctrl {
  controlType=EULER_TRANSFORM
  parent=dyn_hair_bk_l_05_ctrl_Null
  type=Control
  loc=<0.6752 -11.3965 125.4611>
  rot=<-0.5271 -0.5623 0.4495 -0.4515>
 }
 dyn_hair_bk_l_06_ctrl_Null {type=Null, loc=<7.7014 -10.0282 118.2738>, rot=<-0.4772 -0.512 0.5022 -0.5078>}
 dyn_hair_bk_l_06_ctrl {
  controlType=EULER_TRANSFORM
  parent=dyn_hair_bk_l_06_ctrl_Null
  type=Control
  loc=<24.9623 -10.0282 115.6102>
  rot=<0.4221 0.5623 -0.5493 0.4515>
 }
 dyn_hair_bk_r_01_ctrl_Null {type=Null, loc=<-6.25 -6.7546 153.204>, rot=<-0.3454 -0.4116 0.6461 -0.5421>}
 dyn_hair_bk_r_01_ctrl {
  controlType=EULER_TRANSFORM
  parent=dyn_hair_bk_r_01_ctrl_Null
  type=Control
  loc=<-6.25 -6.7546 153.204>
  rot=<0.3454 0.4116 -0.6461 0.5421>
 }
 dyn_hair_bk_r_02_ctrl_Null {type=Null, loc=<-7.5212 -9.8013 146.6702>, rot=<-0.3787 -0.4394 0.6271 -0.5198>}
 dyn_hair_bk_r_02_ctrl {
  controlType=EULER_TRANSFORM
  parent=dyn_hair_bk_r_02_ctrl_Null
  type=Control
  loc=<-30.0915 -9.8013 144.3154>
  rot=<0.411 0.4116 -0.6064 0.5421>
 }
 dyn_hair_bk_r_03_ctrl_Null {type=Null, loc=<-8.7853 -12.1376 139.8496>, rot=<-0.465 -0.4887 0.5561 -0.4855>}
 dyn_hair_bk_r_03_ctrl {
  controlType=EULER_TRANSFORM
  parent=dyn_hair_bk_r_03_ctrl_Null
  type=Control
  loc=<-50.2427 -10.9846 131.604>
  rot=<0.548 0.4119 -0.5097 0.5198>
 }
 dyn_hair_bk_r_04_ctrl_Null {type=Null, loc=<-9.4858 -12.7596 132.6216>, rot=<-0.5623 -0.5271 0.4515 -0.4495>}
 dyn_hair_bk_r_04_ctrl {
  controlType=EULER_TRANSFORM
  parent=dyn_hair_bk_r_04_ctrl_Null
  type=Control
  loc=<-54.0544 -10.6202 122.4663>
  rot=<0.6542 0.4284 -0.3909 0.4855>
 }
 dyn_hair_bk_r_05_ctrl_Null {type=Null, loc=<-9.2191 -11.3965 125.4611>, rot=<-0.5623 -0.5271 0.4515 -0.4495>}
 dyn_hair_bk_r_05_ctrl {
  controlType=EULER_TRANSFORM
  parent=dyn_hair_bk_r_05_ctrl_Null
  type=Control
  loc=<-16.513 -11.3965 125.4611>
  rot=<0.5623 0.5271 -0.4515 0.4495>
 }
 dyn_hair_bk_r_06_ctrl_Null {type=Null, loc=<-8.9514 -10.0282 118.2738>, rot=<-0.512 -0.4772 0.5078 -0.5022>}
 dyn_hair_bk_r_06_ctrl {
  controlType=EULER_TRANSFORM
  parent=dyn_hair_bk_r_06_ctrl_Null
  type=Control
  loc=<8.6733 -10.0282 119.0725>
  rot=<0.4562 0.5271 -0.5586 0.4495>
 }
]
variables [
 Frontend {type=bool, default=False}
 IsCrouching {type=bool, default=False}
 PlayingEmote {type=bool, default=False}
 Skydive {type=bool, default=False}
 wearingBP {type=bool, default=False}
]
