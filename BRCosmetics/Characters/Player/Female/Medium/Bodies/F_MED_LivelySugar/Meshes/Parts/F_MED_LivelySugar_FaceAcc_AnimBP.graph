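# Post-process style AnimBlueprint for LivelySugar's face accessory mesh, animating hair and earrings on top of the copied parent pose.
# A cached pose of hair constraints and earring/strand AnimDynamics feeds either a jacket-variant constraint and RigidBody chain, or Control Rig plus RigidBody chains split by backbling.
# Additive bone offsets correct each branch.
# Update logic selects the jacket by cosmetic tag, computes gravity override, and adjusts alphas for movement, melee and front-end emotes.
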AnimBlueprint /BRCosmetics/Characters/Player/Female/Medium/Bodies/F_MED_LivelySugar/Meshes/Parts/F_MED_LivelySugar_FaceAcc_AnimBP
target_skeleton=&/F_MED_LivelySugar_FaceAcc_Skeleton
graphs [
 AnimGraph {nodes [
   Copy Pose From Mesh() {bUseAttachedParent=true}
   n1 = "Constraint - dyn_hair_bk_01 (PP)"!(
    ComponentPose=Local To Component(LocalPose=Use cached pose 'Constraints Pose'())
    Alpha=1
    ConstraintWeights_0=0.4
    ConstraintWeights_1:float=0.6
   ) -> (Pose) {
    boneToModify {boneName=dyn_hair_bk_01}
    constraintSetup [{targetBone {boneName=head}}, {targetBone {boneName=spine_05}}]
    constraintWeights [1,0]
   }
   n2 = "Constraint - dyn_hair_bk_l_01 (PP)"!(ComponentPose=@n1, Alpha=1, ConstraintWeights_0=0.4, ConstraintWeights_1:float=0.6) -> (Pose) {
    boneToModify {boneName=dyn_hair_bk_l_01}
    constraintSetup [{targetBone {boneName=head}}, {targetBone {boneName=spine_05}}]
    constraintWeights [1,0]
   }
   n3 = "Constraint - dyn_hair_bk_r_01 (PP)"!(ComponentPose=@n2, Alpha=1, ConstraintWeights_0=0.4, ConstraintWeights_1:float=0.6) -> (Pose) {
    boneToModify {boneName=dyn_hair_bk_r_01}
    constraintSetup [{targetBone {boneName=head}}, {targetBone {boneName=spine_05}}]
    constraintWeights [1,0]
   }
   n4 = "Constraint - dyn_hair_bk_r_02 (PPP)"!(ComponentPose=@n3, Alpha=1, ConstraintWeights_0=0.9485, ConstraintWeights_1:float=0.0309, ConstraintWeights_2:float=0.0206) -> (Pose) {
    boneToModify {boneName=dyn_hair_bk_r_02}
    constraintSetup [{targetBone {boneName=spine_05}}, {targetBone {boneName=neck_02}}, {targetBone {boneName=neck_01}}]
    constraintWeights [1,1,1]
   }
   n5 = "Constraint - dyn_hair_bk_02 (PPPP)"!(
    ComponentPose=@n4
    Alpha=1
    ConstraintWeights_0=0.4695
    ConstraintWeights_1:float=0.2098
    ConstraintWeights_2:float=0.1942
    ConstraintWeights_3:float=0.1265
   ) -> (Pose) {
    boneToModify {boneName=dyn_hair_bk_02}
    constraintSetup [{targetBone {boneName=neck_02}}, {targetBone {boneName=neck_01}}, {targetBone {boneName=head}}, {targetBone {boneName=spine_05}}]
    constraintWeights [1,1,1,1]
   }
   n6 = "Constraint - dyn_hair_bk_l_02 (PPP)"!(ComponentPose=@n5, Alpha=1, ConstraintWeights_0=0.949, ConstraintWeights_1:float=0.0306, ConstraintWeights_2:float=0.0204) -> (Pose) {
    boneToModify {boneName=dyn_hair_bk_l_02}
    constraintSetup [{targetBone {boneName=spine_05}}, {targetBone {boneName=neck_02}}, {targetBone {boneName=neck_01}}]
    constraintWeights [1,1,1]
   }
   n7 = "Constraint - dyn_hair_bk_03 (PPP)"!(ComponentPose=@n6, Alpha=1, ConstraintWeights_0=0.875, ConstraintWeights_1:float=0.0625, ConstraintWeights_2:float=0.0625) -> (Pose) {
    boneToModify {boneName=dyn_hair_bk_03}
    constraintSetup [{targetBone {boneName=spine_05}}, {targetBone {boneName=clavicle_r}}, {targetBone {boneName=clavicle_l}}]
    constraintWeights [1,1,1]
   }
   n8 = "Constraint - dyn_hair_bk_04 (PPPP)"!(
    ComponentPose=@n7
    Alpha=1
    ConstraintWeights_0=0.703
    ConstraintWeights_1:float=0.138
    ConstraintWeights_2:float=0.138
    ConstraintWeights_3:float=0.021
   ) -> (Pose) {
    boneToModify {boneName=dyn_hair_bk_04}
    constraintSetup [{targetBone {boneName=spine_05}}, {targetBone {boneName=clavicle_l}}, {targetBone {boneName=clavicle_r}}, {targetBone {boneName=spine_04}}]
    constraintWeights [1,1,1,1]
   }
   n9 = "Constraint - dyn_hair_bk_05 (PPPPPPP)"!(
    ComponentPose=@n8
    Alpha=1
    ConstraintWeights_0=0.562
    ConstraintWeights_1:float=0.191
    ConstraintWeights_2:float=0.069
    ConstraintWeights_3:float=0.057
    ConstraintWeights_4:float=0.057
    ConstraintWeights_5:float=0.046
    ConstraintWeights_6:float=0.018
   ) -> (Pose) {
    boneToModify {boneName=dyn_hair_bk_05}
    constraintSetup [
     {targetBone {boneName=spine_04}}
     {targetBone {boneName=spine_05}}
     {targetBone {boneName=spine_03}}
     {targetBone {boneName=clavicle_r}}
     {targetBone {boneName=clavicle_l}}
     {targetBone {boneName=spine_02}}
     {targetBone {boneName=spine_01}}
    ]
    constraintWeights [1,1,1,1,1,1,1]
   }
   n10 = "Constraint - dyn_hair_bk_06 (PPPPP)"!(
    ComponentPose=@n9
    Alpha=1
    ConstraintWeights_0=0.3445
    ConstraintWeights_1:float=0.2835
    ConstraintWeights_2:float=0.249
    ConstraintWeights_3:float=0.0965
    ConstraintWeights_4:float=0.0264
   ) -> (Pose) {
    boneToModify {boneName=dyn_hair_bk_06}
    constraintSetup [
     {targetBone {boneName=spine_03}}
     {targetBone {boneName=spine_04}}
     {targetBone {boneName=spine_02}}
     {targetBone {boneName=spine_01}}
     {targetBone {boneName=spine_05}}
    ]
    constraintWeights [1,1,1,1,1]
   }
   n11 = "Constraint - dyn_hair_bk_07 (PPPP)"!(
    ComponentPose=@n10
    Alpha=1
    ConstraintWeights_0=0.35
    ConstraintWeights_1:float=0.296
    ConstraintWeights_2:float=0.281
    ConstraintWeights_3:float=0.073
   ) -> (Pose) {
    boneToModify {boneName=dyn_hair_bk_07}
    constraintSetup [{targetBone {boneName=spine_02}}, {targetBone {boneName=spine_01}}, {targetBone {boneName=pelvis}}, {targetBone {boneName=spine_03}}]
    constraintWeights [1,1,1,1]
   }
   n12 = "Constraint - dyn_hair_bk_l_03 (PP)"!(ComponentPose=@n11, Alpha=1, ConstraintWeights_0=0.6, ConstraintWeights_1:float=0.4) -> (Pose) {
    boneToModify {boneName=dyn_hair_bk_l_03}
    constraintSetup [{targetBone {boneName=clavicle_l}}, {targetBone {boneName=spine_05}}]
    constraintWeights [1,1]
   }
   n13 = "Constraint - dyn_hair_bk_l_04 (PPP)"!(ComponentPose=@n12, Alpha=1, ConstraintWeights_0=0.6, ConstraintWeights_1:float=0.2, ConstraintWeights_2:float=0.2) -> (Pose) {
    boneToModify {boneName=dyn_hair_bk_l_04}
    constraintSetup [{targetBone {boneName=clavicle_l}}, {targetBone {boneName=spine_05}}, {targetBone {boneName=spine_04}}]
    constraintWeights [1,1,1]
   }
   n14 = "Constraint - dyn_hair_bk_l_05 (PPPP)"!(
    ComponentPose=@n13
    Alpha=1
    ConstraintWeights_0=0.4
    ConstraintWeights_1:float=0.3
    ConstraintWeights_2:float=0.2
    ConstraintWeights_3:float=0.1
   ) -> (Pose) {
    boneToModify {boneName=dyn_hair_bk_l_05}
    constraintSetup [{targetBone {boneName=spine_04}}, {targetBone {boneName=clavicle_l}}, {targetBone {boneName=spine_05}}, {targetBone {boneName=spine_03}}]
    constraintWeights [1,1,1,1]
   }
   n15 = "Constraint - dyn_hair_bk_l_06 (PPPPPP)"!(
    ComponentPose=@n14
    Alpha=1
    ConstraintWeights_0=0.3691
    ConstraintWeights_1:float=0.2839
    ConstraintWeights_2:float=0.199
    ConstraintWeights_3:float=0.082
    ConstraintWeights_4:float=0.042
    ConstraintWeights_5:float=0.024
   ) -> (Pose) {
    boneToModify {boneName=dyn_hair_bk_l_06}
    constraintSetup [
     {targetBone {boneName=spine_04}}
     {targetBone {boneName=spine_03}}
     {targetBone {boneName=spine_02}}
     {targetBone {boneName=spine_01}}
     {targetBone {boneName=clavicle_l}}
     {targetBone {boneName=spine_05}}
    ]
    constraintWeights [1,1,1,1,1,1]
   }
   n16 = "Constraint - dyn_hair_bk_r_03 (PP)"!(ComponentPose=@n15, Alpha=1, ConstraintWeights_0=0.8, ConstraintWeights_1:float=0.2) -> (Pose) {
    boneToModify {boneName=dyn_hair_bk_r_03}
    constraintSetup [{targetBone {boneName=clavicle_r}}, {targetBone {boneName=spine_05}}]
    constraintWeights [1,1]
   }
   n17 = "Constraint - dyn_hair_bk_r_04 (PPP)"!(ComponentPose=@n16, Alpha=1, ConstraintWeights_0=0.7, ConstraintWeights_1:float=0.2, ConstraintWeights_2:float=0.1) -> (Pose) {
    boneToModify {boneName=dyn_hair_bk_r_04}
    constraintSetup [{targetBone {boneName=clavicle_r}}, {targetBone {boneName=spine_05}}, {targetBone {boneName=spine_04}}]
    constraintWeights [1,1,1]
   }
   n18 = "Constraint - dyn_hair_bk_r_05 (PPPP)"!(
    ComponentPose=@n17
    Alpha=1
    ConstraintWeights_0=0.4
    ConstraintWeights_1:float=0.35
    ConstraintWeights_2:float=0.15
    ConstraintWeights_3:float=0.1
   ) -> (Pose) {
    boneToModify {boneName=dyn_hair_bk_r_05}
    constraintSetup [{targetBone {boneName=spine_04}}, {targetBone {boneName=clavicle_r}}, {targetBone {boneName=spine_05}}, {targetBone {boneName=spine_03}}]
    constraintWeights [1,1,1,1]
   }
   n19 = "Constraint - dyn_hair_bk_r_06 (PPPPPP)"!(
    ComponentPose=@n18
    Alpha=1
    ConstraintWeights_0=0.589
    ConstraintWeights_1:float=0.222
    ConstraintWeights_2:float=0.072
    ConstraintWeights_3:float=0.067
    ConstraintWeights_4:float=0.026
    ConstraintWeights_5:float=0.024
   ) -> (Pose) {
    boneToModify {boneName=dyn_hair_bk_r_06}
    constraintSetup [
     {targetBone {boneName=spine_04}}
     {targetBone {boneName=spine_03}}
     {targetBone {boneName=spine_02}}
     {targetBone {boneName=clavicle_r}}
     {targetBone {boneName=spine_05}}
     {targetBone {boneName=spine_01}}
    ]
    constraintWeights [1,1,1,1,1,1]
   }
   n20 = "Constraint - dyn_hair_l_01 (P)"(ComponentPose=@n19, Alpha=1, ConstraintWeights_0=1) {boneToModify {boneName=dyn_hair_l_01}, constraintSetup [{targetBone {boneName=spine_05}}], constraintWeights [1]}
   n21 = "Constraint - dyn_hair_r_01 (P)"(ComponentPose=@n20, Alpha=1, ConstraintWeights_0=1) {boneToModify {boneName=dyn_hair_r_01}, constraintSetup [{targetBone {boneName=spine_05}}], constraintWeights [1]}
   n22 = "Transform (Modify) Bone - Bone: dyn_hair_bk_l_02"(ComponentPose=@n21, Alpha=1, Translation="0,-1.8,-2.5", Rotation="0,0,0", Scale="1,1,1") {
    rotationMode=BMM_Additive
    rotationSpace=BCS_BoneSpace
    translationMode=BMM_Additive
    translationSpace=BCS_BoneSpace
    boneToModify {boneName=dyn_hair_bk_l_02}
    translation {y=-1.8, z=-2.5}
   }
   n23 = "Transform (Modify) Bone - Bone: dyn_hair_bk_03"(ComponentPose=@n22, Alpha=1, Translation="0,0,0", Rotation="-3.7,0,0", Scale="1,1,1") {
    rotationMode=BMM_Additive
    rotationSpace=BCS_BoneSpace
    translationMode=BMM_Additive
    translationSpace=BCS_BoneSpace
    boneToModify {boneName=dyn_hair_bk_03}
    rotation {pitch=-3.7}
   }
   n24 = "Transform (Modify) Bone - Bone: dyn_hair_bk_r_02"(ComponentPose=@n23, Alpha=1, Translation="0,1.8,-2.5", Rotation="-4.5,5,0", Scale="1,1,1") {
    rotationMode=BMM_Additive
    rotationSpace=BCS_BoneSpace
    translationMode=BMM_Additive
    translationSpace=BCS_BoneSpace
    boneToModify {boneName=dyn_hair_bk_r_02}
    rotation {pitch=-4.5, yaw=5}
    translation {y=1.8, z=-2.5}
   }
   n25 = "Transform (Modify) Bone - Bone: dyn_hair_bk_02"(ComponentPose=@n24, Alpha=1, Translation="0.022,-0.146,-0.363", Rotation="0,0,0", Scale="1,1,0.8") {
    rotationMode=BMM_Additive
    rotationSpace=BCS_BoneSpace
    scaleMode=BMM_Additive
    scaleSpace=BCS_BoneSpace
    translationMode=BMM_Additive
    translationSpace=BCS_BoneSpace
    boneToModify {boneName=dyn_hair_bk_02}
    scale {z=0.8}
    translation=<0.022 -0.146 -0.363>
   }
   n26 = "Transform (Modify) Bone - Bone: dyn_hair_bk_05"(ComponentPose=@n25, Alpha=1, Translation="0,0,1.5", Rotation="0,0,0", Scale="1,1,1") {
    rotationMode=BMM_Additive
    rotationSpace=BCS_BoneSpace
    scaleMode=BMM_Additive
    scaleSpace=BCS_BoneSpace
    translationMode=BMM_Additive
    translationSpace=BCS_BoneSpace
    boneToModify {boneName=dyn_hair_bk_05}
    translation {z=1.5}
   }
   n27 = "Transform (Modify) Bone - Bone: dyn_hair_bk_l_02"(ComponentPose=@n26, Alpha=1, Translation="0,0,-2.5", Rotation="0,0,0", Scale="1,1,0.8") {
    rotationMode=BMM_Additive
    rotationSpace=BCS_BoneSpace
    scaleMode=BMM_Additive
    scaleSpace=BCS_BoneSpace
    translationMode=BMM_Additive
    translationSpace=BCS_BoneSpace
    boneToModify {boneName=dyn_hair_bk_l_02}
    scale {z=0.8}
    translation {z=-2.5}
   }
   n28 = "Transform (Modify) Bone - Bone: dyn_hair_r_01"(ComponentPose=@n27, Alpha=1, Translation="0,0,0", Rotation="-0.688,-12.899,-15.580", Scale="1,1,1") {
    rotationMode=BMM_Additive
    rotationSpace=BCS_BoneSpace
    scaleMode=BMM_Additive
    scaleSpace=BCS_BoneSpace
    translationMode=BMM_Additive
    translationSpace=BCS_BoneSpace
    boneToModify {boneName=dyn_hair_r_01}
    rotation=<rot -0.688 -12.899 -15.58>
   }
   n29 = "Transform (Modify) Bone - Bone: dyn_hair_l_01"(ComponentPose=@n28, Alpha=1, Translation="0,0,0", Rotation="4.556,-10.064,19.297", Scale="1,1,1") {
    rotationMode=BMM_Additive
    rotationSpace=BCS_BoneSpace
    scaleMode=BMM_Additive
    scaleSpace=BCS_BoneSpace
    translationMode=BMM_Additive
    translationSpace=BCS_BoneSpace
    boneToModify {boneName=dyn_hair_l_01}
    rotation=<rot 4.556 -10.064 19.297>
   }
   n30 = RigidBody!(ComponentPose=@n29, Alpha, OverrideWorldGravity:Vector=Get GravityOverride(), ExternalForce) -> (Pose) {
    bOverrideWorldGravity=true
    overridePhysicsAsset=&/F_MED_FaceAcc_LivelySugar_Jacket_Physics
    simulationSpace=BaseBoneSpace
    baseBoneRef {boneName=spine_05}
    simSpaceSettings {
     maxAngularAcceleration=80
     maxAngularVelocity=40
     maxLinearAcceleration=1000
     maxLinearVelocity=400
     worldAlpha=1
     externalLinearDragV=<0.4 0.4 0.4>
    }
   }
   n31 = "Transform (Modify) Bone - Bone: dyn_hair_r_01"!(ComponentPose=@n30, bAlphaBoolEnabled:bool=True, Rotation="7.999,3.9963,-14.0103") -> (Pose) {
    alphaInputType=Bool
    rotationMode=BMM_Additive
    rotationSpace=BCS_BoneSpace
    scaleMode=BMM_Additive
    scaleSpace=BCS_BoneSpace
    translationMode=BMM_Additive
    translationSpace=BCS_BoneSpace
    boneToModify {boneName=dyn_hair_r_01}
    rotation=<rot 7.999 3.9963 -14.0103>
   }
   n32 = "Transform (Modify) Bone - Bone: dyn_hair_l_05"!(ComponentPose=@n31, bAlphaBoolEnabled:bool=True, Rotation="-2.3734,24.4081,7.3607") -> (Pose) {
    alphaInputType=Bool
    rotationMode=BMM_Additive
    rotationSpace=BCS_BoneSpace
    scaleMode=BMM_Additive
    scaleSpace=BCS_BoneSpace
    translationMode=BMM_Additive
    translationSpace=BCS_BoneSpace
    boneToModify {boneName=dyn_hair_l_05}
    rotation=<rot -2.3734 24.4081 7.3607>
   }
   Output Pose(Result=Blend Poses by bool(
     BlendPose_0=Component To Local(ComponentPose=@n32)
     BlendPose_1=Use cached pose 'Dynamics'()
     bActiveValue=Get Jacket vis() -> (Jacket vis:bool)
    )) {layerGroup=None}
   n33 = "Constraint - dyn_hair_l_01 (PP)"!(
    ComponentPose=Local To Component(LocalPose=Copy Pose From Mesh() {bUseAttachedParent=true})
    bAlphaBoolEnabled:bool=True
    ConstraintWeights_0=0.01
    ConstraintWeights_1:float=0.99
   ) -> (Pose) {
    alphaInputType=Bool
    boneToModify {boneName=dyn_hair_l_01}
    constraintSetup [{targetBone {boneName=spine_05}}, {targetBone {boneName=neck_01}}]
    constraintWeights [1,1]
   }
   n34 = "Constraint - dyn_hair_r_01 (PP)"!(ComponentPose=@n33, bAlphaBoolEnabled:bool=True, ConstraintWeights_0=0.2, ConstraintWeights_1:float=0.8) -> (Pose) {
    alphaInputType=Bool
    boneToModify {boneName=dyn_hair_r_01}
    constraintSetup [{targetBone {boneName=spine_05}}, {targetBone {boneName=neck_01}}]
    constraintWeights [1,1]
   }
   n35 = "Constraint - dyn_hair_l_02 (PP)"!(ComponentPose=@n34, bAlphaBoolEnabled:bool=True, ConstraintWeights_0=0.1, ConstraintWeights_1:float=0.9) -> (Pose) {
    alphaInputType=Bool
    boneToModify {boneName=dyn_hair_l_02}
    constraintSetup [{targetBone {boneName=spine_05}}, {targetBone {boneName=neck_01}}]
    constraintWeights [1,1]
   }
   n36 = "Constraint - dyn_hair_r_02 (PP)"!(ComponentPose=@n35, bAlphaBoolEnabled:bool=True, ConstraintWeights_0=0.2, ConstraintWeights_1:float=0.8) -> (Pose) {
    alphaInputType=Bool
    boneToModify {boneName=dyn_hair_r_02}
    constraintSetup [{targetBone {boneName=spine_05}}, {targetBone {boneName=neck_01}}]
    constraintWeights [1,1]
   }
   n37 = "Constraint - dyn_hair_l_03 (PP)"!(ComponentPose=@n36, bAlphaBoolEnabled:bool=True, ConstraintWeights_0=0.4, ConstraintWeights_1:float=0.6) -> (Pose) {
    alphaInputType=Bool
    boneToModify {boneName=dyn_hair_l_03}
    constraintSetup [{targetBone {boneName=spine_05}}, {targetBone {boneName=neck_01}}]
    constraintWeights [1,1]
   }
   n38 = "Constraint - dyn_hair_r_03 (PP)"!(ComponentPose=@n37, bAlphaBoolEnabled:bool=True, ConstraintWeights_0=0.4, ConstraintWeights_1:float=0.6) -> (Pose) {
    alphaInputType=Bool
    boneToModify {boneName=dyn_hair_r_03}
    constraintSetup [{targetBone {boneName=spine_05}}, {targetBone {boneName=neck_01}}]
    constraintWeights [1,1]
   }
   n39 = "Constraint - dyn_hair_l_04 (PP)"!(ComponentPose=@n38, bAlphaBoolEnabled:bool=True, ConstraintWeights_0=0.8, ConstraintWeights_1:float=0.2) -> (Pose) {
    alphaInputType=Bool
    boneToModify {boneName=dyn_hair_l_04}
    constraintSetup [{targetBone {boneName=spine_05}}, {targetBone {boneName=neck_01}}]
    constraintWeights [1,1]
   }
   n40 = "Constraint - dyn_hair_r_04 (PP)"!(ComponentPose=@n39, bAlphaBoolEnabled:bool=True, ConstraintWeights_0=0.8, ConstraintWeights_1:float=0.2) -> (Pose) {
    alphaInputType=Bool
    boneToModify {boneName=dyn_hair_r_04}
    constraintSetup [{targetBone {boneName=spine_05}}, {targetBone {boneName=neck_01}}]
    constraintWeights [1,1]
   }
   n41 = Anim Dynamics - Bone: dyn_earring_r(ComponentPose=@n40, Alpha=1, ExternalForce="0,0,0") {
    angularDampingOverride=0.999
    angularSpringConstant=100
    bAngularSpring=true
    bOverrideAngularDamping=true
    bUseGravityOverride=true
    gravityScale=0.35
    boundBone {boneName=dyn_earring_r}
    componentAppliedLinearAccClamp=<6000 6000 6000>
    componentLinearAccScale=<0.8 0.8 0.8>
    componentLinearVelScale=<0.8 0.8 0.8>
    gravityOverride {z=980}
    physicsBodyDefinitions [{
      boundBone {boneName=dyn_earring_r}
      boxExtents=<20 30 30>
      constraintSetup {angularLimitsMax=<0 0 50>, angularLimitsMin=<0 -10 -20>, angularTarget=<1 0 0>}
      localJointOffset=<3.7 0 0>
     }]
   }
   n42 = Anim Dynamics - Bone: dyn_earring_l(ComponentPose=@n41, Alpha=1, ExternalForce="0,0,0") {
    angularDampingOverride=0.999
    angularSpringConstant=100
    bAngularSpring=true
    bOverrideAngularDamping=true
    bUseGravityOverride=true
    gravityScale=0.35
    boundBone {boneName=dyn_earring_l}
    componentAppliedLinearAccClamp=<6000 6000 6000>
    componentLinearAccScale=<0.8 0.8 0.8>
    componentLinearVelScale=<0.8 0.8 0.8>
    gravityOverride {z=980}
    physicsBodyDefinitions [{
      boundBone {boneName=dyn_earring_l}
      boxExtents=<20 30 30>
      constraintSetup {angularLimitsMax=<0 10 50>, angularLimitsMin=<0 0 -20>, angularTarget=<1 0 0>}
      localJointOffset=<3.7 0 0>
     }]
   }
   n43 = Anim Dynamics - Bone: dyn_hairStrand_03_l(ComponentPose=@n42, Alpha=1, ExternalForce="0,0,0") {
    angularDampingOverride=0.9999
    angularSpringConstant=5000
    bAngularSpring=true
    bOverrideAngularDamping=true
    bUseGravityOverride=true
    gravityScale=0.35
    boundBone {boneName=dyn_hairStrand_03_l}
    componentLinearAccScale=<1 1 1>
    componentLinearVelScale=<1 1 1>
    gravityOverride {z=-980}
    physicsBodyDefinitions [{
      boundBone {boneName=dyn_hairStrand_03_l}
      boxExtents=<20 30 30>
      constraintSetup {angularLimitsMax=<20 0 25>, angularLimitsMin=<0 0 -25>, angularTarget=<1 0 0>}
      localJointOffset=<4 0 0>
     }]
   }
   n44 = Anim Dynamics - Bone: dyn_hairStrand_02_l(ComponentPose=@n43, Alpha=1, ExternalForce="0,0,0") {
    angularDampingOverride=0.9999
    angularSpringConstant=5000
    bAngularSpring=true
    bOverrideAngularDamping=true
    bUseGravityOverride=true
    gravityScale=0.35
    boundBone {boneName=dyn_hairStrand_02_l}
    componentLinearAccScale=<1 1 1>
    componentLinearVelScale=<1 1 1>
    gravityOverride {z=-980}
    physicsBodyDefinitions [{
      boundBone {boneName=dyn_hairStrand_02_l}
      boxExtents=<20 30 30>
      constraintSetup {angularLimitsMax=<20 0 25>, angularLimitsMin=<0 0 -25>, angularTarget=<1 0 0>}
      localJointOffset=<4 0 0>
     }]
   }
   n45 = Anim Dynamics - Bone: dyn_hairStrand_01_l(ComponentPose=@n44, Alpha=1, ExternalForce="0,0,0") {
    angularDampingOverride=0.9999
    angularSpringConstant=5000
    bAngularSpring=true
    bOverrideAngularDamping=true
    bUseGravityOverride=true
    gravityScale=0.35
    boundBone {boneName=dyn_hairStrand_01_l}
    componentLinearAccScale=<1 1 1>
    componentLinearVelScale=<1 1 1>
    gravityOverride {z=-980}
    physicsBodyDefinitions [{
      boundBone {boneName=dyn_hairStrand_01_l}
      boxExtents=<20 30 30>
      constraintSetup {angularLimitsMax=<20 0 25>, angularLimitsMin=<0 0 -25>, angularTarget=<1 0 0>}
      localJointOffset=<4 0 0>
     }]
   }
   Save cached pose 'Constraints Pose'(Pose=Component To Local(ComponentPose=@n45)) {cacheName=Constraints Pose}
   n46 = Get bIsInFrontEnd()
   n47 = Get bIsSkyDiving()
   n48 = Get isWearingPack() -> (isWearingPack:bool)
   n49 = Control Rig!(
    Source=Use cached pose 'Constraints Pose'()
    Alpha
    Skydive:bool=@n47
    wearingBP:bool=@n48
    PlayingEmote:bool=Get bIsPlayingEmote()
    Frontend:bool=@n46
    IsCrouching:bool=Get bIsCrouching()
   ) -> (Pose) {controlRigAssetReference {blueprintRigClass=&/F_MED_LivelySugar_FaceAcc_CtrlRig.F_MED_LivelySugar_FaceAcc_CtrlRig_C}}
   n50 = "Transform (Modify) Bone - Bone: dyn_hair_l_01"!(ComponentPose=Local To Component(LocalPose=@n49), bAlphaBoolEnabled:bool=True, Translation="0,0,0", Rotation="4,0,-16", Scale="1,1,1") -> (Pose) {
    alphaInputType=Bool
    rotationMode=BMM_Additive
    rotationSpace=BCS_BoneSpace
    scaleMode=BMM_Additive
    scaleSpace=BCS_BoneSpace
    translationMode=BMM_Additive
    translationSpace=BCS_BoneSpace
    boneToModify {boneName=dyn_hair_l_01}
    rotation {pitch=4, roll=-16}
   }
   n51 = RigidBody!(
    ComponentPose=@n50
    Alpha=Get alpha hairRigidBody() -> (alpha hairRigidBody:double)
    OverrideWorldGravity=Get GravityOverride()
    ExternalForce
    SimSpaceSettings:Sim Space Settings Structure=(WorldAlpha=1,VelocityScaleZ=1,DampingAlpha=1,MaxLinearVelocity=400,MaxAngularVelocity=3,MaxLinearAcceleration=800,MaxAngularAcceleration=80,ExternalLinearDragV=(X=0.4,Y=0.4,Z=0.4),ExternalLinearVelocity=(X=0,Y=0,Z=0),ExternalAngularVelocity=(X=0,Y=0,Z=0))
   ) -> (Pose) {
    bOverrideWorldGravity=true
    overridePhysicsAsset=&/F_MED_FaceAcc_LivelySugar_Physics
    simulationSpace=BaseBoneSpace
    alphaScaleBiasClamp {bInterpResult=true, interpSpeedDecreasing=1}
    baseBoneRef {boneName=neck_01}
    simSpaceSettings {
     maxAngularAcceleration=80
     maxAngularVelocity=3
     maxLinearAcceleration=800
     maxLinearVelocity=400
     worldAlpha=1
     externalLinearDragV=<0.4 0.4 0.4>
    }
   }
   n52 = "Transform (Modify) Bone - Bone: dyn_hair_bk_l_02"!(ComponentPose=@n51, bAlphaBoolEnabled:bool=True, Translation="0,1,2", Rotation="0,0,0", Scale="1,1,1") -> (Pose) {
    alphaInputType=Bool
    translationMode=BMM_Additive
    translationSpace=BCS_BoneSpace
    alphaBoolBlend {blendInTime=0.25, blendOutTime=0.25}
    boneToModify {boneName=dyn_hair_bk_l_02}
    translation {y=1, z=2}
   }
   n53 = "Transform (Modify) Bone - Bone: dyn_hair_bk_r_02"!(ComponentPose=@n52, bAlphaBoolEnabled:bool=True, Translation="0,-0.5,2", Rotation="0,0,0", Scale="1,1,1") -> (Pose) {
    alphaInputType=Bool
    translationMode=BMM_Additive
    translationSpace=BCS_BoneSpace
    alphaBoolBlend {blendInTime=0.25, blendOutTime=0.25}
    boneToModify {boneName=dyn_hair_bk_r_02}
    translation {y=-0.5, z=2}
   }
   n54 = "Transform (Modify) Bone - Bone: dyn_hair_bk_l_05"(ComponentPose=@n53, Alpha=1, Translation="0,0,0", Rotation="6.067,5.027,4.608", Scale="1,1,1") {rotationMode=BMM_Additive, rotationSpace=BCS_BoneSpace, boneToModify {boneName=dyn_hair_bk_l_05}, rotation=<rot 6.067 5.027 4.608>}
   n55 = "Transform (Modify) Bone - Bone: dyn_hair_l_03"!(ComponentPose=@n54, bAlphaBoolEnabled:bool=True, Translation="0,-0.7,0", Rotation="0,0,0", Scale="1,1,1") -> (Pose) {
    alphaInputType=Bool
    rotationMode=BMM_Additive
    rotationSpace=BCS_BoneSpace
    scaleMode=BMM_Additive
    scaleSpace=BCS_BoneSpace
    translationMode=BMM_Additive
    translationSpace=BCS_BoneSpace
    boneToModify {boneName=dyn_hair_l_03}
    translation {y=-0.7}
   }
   n56 = "Transform (Modify) Bone - Bone: dyn_hair_l_01"(ComponentPose=@n55, bAlphaBoolEnabled=True, Translation="-1.2,0.026,0.002", Rotation="-1.801,-0.046,0.125", Scale="1,1,1") {
    alphaInputType=Bool
    rotationMode=BMM_Additive
    rotationSpace=BCS_BoneSpace
    scaleMode=BMM_Additive
    scaleSpace=BCS_BoneSpace
    translationMode=BMM_Additive
    translationSpace=BCS_BoneSpace
    boneToModify {boneName=dyn_hair_l_01}
    rotation=<rot -1.801 -0.046 0.125>
    translation=<-1.2 0.026 0.002>
   }
   n57 = "Transform (Modify) Bone - Bone: dyn_hair_l_04"!(ComponentPose=@n56, bAlphaBoolEnabled:bool=True, Translation="0,0,0", Rotation="15.149,2.192,3.209", Scale="1,1,1") -> (Pose) {
    alphaInputType=Bool
    rotationMode=BMM_Additive
    rotationSpace=BCS_BoneSpace
    scaleMode=BMM_Additive
    scaleSpace=BCS_BoneSpace
    translationMode=BMM_Additive
    translationSpace=BCS_BoneSpace
    boneToModify {boneName=dyn_hair_l_04}
    rotation=<rot 15.149 2.192 3.209>
   }
   n58 = "Transform (Modify) Bone - Bone: dyn_hair_l_04"(ComponentPose=@n57, bAlphaBoolEnabled=Get emoteFE() -> (emoteFE:bool), Translation="0,0,0", Rotation="-7.142,-3.132,37.442", Scale="1,1,1") {
    alphaInputType=Bool
    rotationMode=BMM_Additive
    rotationSpace=BCS_BoneSpace
    scaleMode=BMM_Additive
    scaleSpace=BCS_BoneSpace
    translationMode=BMM_Additive
    translationSpace=BCS_BoneSpace
    alphaBoolBlend {blendInTime=0.2, blendOutTime=0.2}
    boneToModify {boneName=dyn_hair_l_04}
    rotation=<rot -7.142 -3.132 37.442>
   }
   n59 = "Constraint - dyn_hair_bk_l_04 (PPP)"(
    ComponentPose=@n58
    Alpha=Get constraintAlpha() -> (constraintAlpha:double)
    ConstraintWeights_0=0.6
    ConstraintWeights_1=0.2
    ConstraintWeights_2=0.2
   ) {
    boneToModify {boneName=dyn_hair_bk_l_04}
    constraintSetup [{targetBone {boneName=clavicle_l}}, {targetBone {boneName=spine_05}}, {targetBone {boneName=spine_04}}]
    constraintWeights [1,1,1]
   }
   n60 = Control Rig!(Source=Use cached pose 'Constraints Pose'(), Alpha, Skydive=@n47, wearingBP=@n48, Frontend=@n46) -> (Pose) {controlRigAssetReference {blueprintRigClass=&F_MED_LivelySugar_FaceAcc_CtrlRig_C}}
   n61 = RigidBody!(ComponentPose=Local To Component(LocalPose=@n60), Alpha=0.85, OverrideWorldGravity=Get GravityOverride(), ExternalForce) -> (Pose) {
    bOverrideWorldGravity=true
    overridePhysicsAsset=&/F_MED_FaceAcc_LivelySugar_BP_Physics
    simulationSpace=BaseBoneSpace
    baseBoneRef {boneName=neck_01}
    simSpaceSettings {worldAlpha=0.4}
   }
   n62 = "Transform (Modify) Bone - Bone: dyn_hair_l_03"(ComponentPose=@n61, bAlphaBoolEnabled=True, Translation="0,-0.7,0", Rotation="0,0,0", Scale="1,1,1") {
    alphaInputType=Bool
    rotationMode=BMM_Additive
    rotationSpace=BCS_BoneSpace
    scaleMode=BMM_Additive
    scaleSpace=BCS_BoneSpace
    translationMode=BMM_Additive
    translationSpace=BCS_BoneSpace
    boneToModify {boneName=dyn_hair_l_03}
    translation {y=-0.7}
   }
   n63 = "Transform (Modify) Bone - Bone: dyn_hair_l_04"(ComponentPose=@n62, bAlphaBoolEnabled=True, Translation="0,0,0", Rotation="15.282,-1.715,2.196", Scale="1,1,1") {
    alphaInputType=Bool
    rotationMode=BMM_Additive
    rotationSpace=BCS_BoneSpace
    scaleMode=BMM_Additive
    scaleSpace=BCS_BoneSpace
    translationMode=BMM_Additive
    translationSpace=BCS_BoneSpace
    boneToModify {boneName=dyn_hair_l_04}
    rotation=<rot 15.282 -1.715 2.196>
   }
   n64 = "Transform (Modify) Bone - Bone: dyn_hair_bk_l_02"(ComponentPose=@n63, bAlphaBoolEnabled=True, Translation="-2.7,0.013,-0.012", Rotation="0,0,0", Scale="1,1,1") {
    alphaInputType=Bool
    rotationMode=BMM_Additive
    rotationSpace=BCS_BoneSpace
    scaleMode=BMM_Additive
    scaleSpace=BCS_BoneSpace
    translationMode=BMM_Additive
    translationSpace=BCS_BoneSpace
    boneToModify {boneName=dyn_hair_bk_l_02}
    translation=<-2.7 0.013 -0.012>
   }
   n65 = "Transform (Modify) Bone - Bone: dyn_hair_l_04"(ComponentPose=@n64, bAlphaBoolEnabled=True, Translation="0,0,0", Rotation="-7.142,-3.132,37.442", Scale="1,1,1") {
    alphaInputType=Bool
    rotationMode=BMM_Additive
    rotationSpace=BCS_BoneSpace
    scaleMode=BMM_Additive
    scaleSpace=BCS_BoneSpace
    translationMode=BMM_Additive
    translationSpace=BCS_BoneSpace
    alphaBoolBlend {blendInTime=0.2, blendOutTime=0.2}
    boneToModify {boneName=dyn_hair_l_04}
    rotation=<rot -7.142 -3.132 37.442>
   }
   Save cached pose 'Dynamics'(Pose=Blend Poses by bool(BlendPose_0=Component To Local(ComponentPose=@n65), BlendPose_1=Component To Local(ComponentPose=@n59), bActiveValue=Get isWearingPack())) {cacheName=Dynamics}
  ]}
 EventGraph {nodes [
   Event BlueprintUpdateAnimation()
   > n1 = Sequence!(execute) -> (then_0, then_1, then_2:Exec, then_3:Exec, then_4:Exec)
   > n2 = Branch(Condition=CharacterItemDefinitionHasAnyMetaTag(Tags=Make Array("[0]"=(TagName="Cosmetics.Variant.Property.Stage1"))))
   > n3 = Set Jacket vis(Jacket vis:bool=true) -> (Output_Get:bool)
   > Set Jacket Alpha(Jacket Alpha:bool=NOT Boolean(A=@n3)) -> (Output_Get:bool)
   Set Jacket vis(execute=@n2.else, Jacket vis=false)
   n4 = Branch(execute=@n1.then_3, Condition=Get bIsPlayingMeleeAnim())
   > Set alpha hairRigidBody(alpha hairRigidBody:double=0.6) -> (Output_Get:double)
   n5 = Branch(execute=@n4.else, Condition=Get isTacticalJumpingFalling() -> (isTacticalJumpingFalling:bool))
   > Set alpha hairRigidBody(alpha hairRigidBody=0.8)
   Set alpha hairRigidBody(execute=@n5.else, alpha hairRigidBody=1)
   n6 = "Gravity Override State Function (Parameter Asset) (Nativized)"(
    execute=@n1.then_1
    CharacterPartAnimInstance=Self-Reference()
    GameTimeInSeconds=GetGameTimeInSeconds()
    DebugJointName=neck_02
    Parameters=&/F_MED_LivelySugar_Parameter
   )
   > Set GravityOverride(GravityOverride=@n6)
   > Set RBAN Settings Out(RBAN Settings Out=@n6.OutRigidBodyAnimNodeParameters)
   n7 = Get bIsTacticalSprinting()
   n8 = Get bIsJumping()
   n9 = Get bIsFalling()
   Set isTacticalJumpingFalling(execute=@n1.then_2, isTacticalJumpingFalling:bool=OR Boolean(A=AND Boolean(A=@n8, B=@n7), B=AND Boolean(A=@n7, B=@n9))) -> (Output_Get:bool)
   > n10 = Branch(Condition=Get bIsInFrontEnd())
   > n11 = Branch(Condition=Get bIsPlayingEmote())
   > Set constraintAlpha(constraintAlpha:double=0) -> (Output_Get:double)
   Set constraintAlpha(execute=@n11.else, constraintAlpha=0.25)
   Set constraintAlpha(execute=@n10.else, constraintAlpha=0)
   Set emoteFE(execute=@n1.then_4, emoteFE:bool=AND Boolean(A=Get bIsInFrontEnd(), B=Get bIsPlayingEmote())) -> (Output_Get:bool)
   Event OnCharacterCustomizationLoaded()
   n12 = CheckForBackbling()
   > Set isWearingPack(isWearingPack:bool=OR Boolean!(A=@n12, B=@n12.bIsBackblingACape, C:bool=@n12.bIsBackblingLong) -> (ReturnValue)) -> (Output_Get:bool)
   > Set isWearingCape(isWearingCape=@n12.bIsBackblingLong)
  ]}
]
variables [
 alpha hairRigidBody {type=number}
 constraintAlpha {type=number}
 emoteFE {type=boolean}
 GravityOverride {type=Vector}
 isTacticalJumpingFalling {type=boolean}
 isWearingCape {type=boolean}
 isWearingPack {type=boolean}
 Jacket Alpha {type=boolean}
 Jacket vis {type=boolean}
 No Skydive Alpha {type=number}
 RBAN Settings Out {type=FortRigidBodyAnimNodeParameters}
]
